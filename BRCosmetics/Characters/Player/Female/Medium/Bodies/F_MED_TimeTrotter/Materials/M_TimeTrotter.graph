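# Skeletal-mesh material for the TimeTrotter female medium character body, adding an animated, masked emissive effect between the shared CharacterParent PreFX and PostFX functions.
# TechArtMap channels supply masks and a gradient coordinate, with two UV-bounded patches substituting shifted procedural ramps.
# Time scrolls that coordinate through a color gradient, sampled with explicit UV derivatives alongside Frac wrapping.
# An attribute blend layers the emissive, with adjustable Fresnel shaping and black base color, onto the character material.
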Material /BRCosmetics/Characters/Player/Female/Medium/Bodies/F_MED_TimeTrotter/Materials/M_TimeTrotter
graph {
 properties {
  bAutomaticallySetUsageInEditor=false
  bUseMaterialAttributes=true
  bUsedWithMeshDeformer=true
  bUsedWithMorphTargets=true
  bUsedWithSkeletalMesh=true
  pixelDepthOffsetMode=PDOM_Legacy
  shadingModel=MSM_FromMaterialExpression
  usedShadingModels=Default Lit | Subsurface | Clear Coat | Subsurface Profile | Cloth
 }
 nodes [
  n1 = NamedRerouteUsage()
  n2 = Constant()
  n3 = NamedRerouteUsage()
  n4 = LinearInterpolate(A=OneMinus(None=@n3), B=@n3, Alpha=ScalarParameter() {group=TimeTrotter, parameterName=Emissive_Fresnel, sortPriority=39})
  n5 = ScalarParameter() {defaultValue=2, group=TimeTrotter, parameterName=Emissive_FresnelPower, sortPriority=39}
  n6 = Multiply(A=NamedRerouteUsage(), B=Saturate(None=Power(Base=@n4, Exp=@n5)))
  n7 = MaterialFunctionCall() {materialFunction=&MF_CharacterParent_2023_PreFX}
  MaterialFunctionCall(
   Material=BlendMaterialAttributes(
    A=NamedRerouteUsage()
    B=SetMaterialAttributes!(MaterialAttributes=@n1, Base Color=@n2, Emissive Color=@n6) {attributeSetTypes [#1, #2]}
    Alpha=NamedRerouteUsage()
   ) {vertexAttributeBlendType=UseA}
   M=@n7.M
   Pre Skinned Local Position=@n7.M
   Distance Blend=@n7.M
   Fresnel=@n7.M
  ) {materialFunction=&MF_CharacterParent_2023_PostFX}
  NamedRerouteDeclaration(None=@n7.Fresnel) {name=Fresnel, nodeColor=<rgba 1 0 0.7734 1>}
  NamedRerouteDeclaration(None=@n7.Result) {name=Material Attributes, nodeColor=<rgba 1 0.5156 0 1>}
  n8 = TextureSampleParameter2D() {
   group=TimeTrotter
   parameterName=TechArtMap
   samplerSource=SSM_Wrap_WorldGroupSettings
   samplerType=SAMPLERTYPE_Masks
   texture=&T_BlackMask
  }
  n9 = TextureCoordinate()
  n10 = ComponentMask(None=@n9) {r=true}
  n11 = Step(Y=@n10) {constX=0.383}
  n12 = Step(Y=@n10) {constX=0.3}
  n13 = ComponentMask(None=@n9) {g=true}
  n14 = Multiply(A=Min(A=Subtract(A=@n11, B=@n12), B=Step(X=@n13) {constX=0.5, constY=0.64}), B=@n8.R)
  n15 = VectorParameter() {parameterName=Patch1, defaultValue=<rgba 0 0.5 -0.45 1>}
  n16 = Saturate(None=MaterialFunctionCall!(Mask In=OneMinus(None=@n10), Amount=@n15.R, Range=@n15.R, Start Offset=@n15.R, Sharpness=@n15.R) {materialFunction=/Game/Global/MaterialFunctions/MF_MaskShifting}.Result)
  n17 = ComponentMask(None=@n9) {r=true}
  n18 = Step(Y=@n17) {constX=0.35}
  n19 = Step(Y=@n17) {constX=0.1}
  n20 = ComponentMask(None=@n9) {g=true}
  n21 = Multiply(A=Multiply(A=Subtract(A=@n18, B=@n19), B=Step(Y=@n20) {constX=0.1}), B=@n8.R)
  n22 = Add(A=@n10) {constB=0.2715}
  n23 = Abs(None=Subtract(A=Multiply(A=@n22) {constB=2}))
  n24 = VectorParameter() {parameterName=Patch2, defaultValue=<rgba 0 0.6 0.05 1>}
  n25 = LinearInterpolate(
   A=LinearInterpolate(A=@n8.B, B=Multiply(A=@n14, B=@n16), Alpha=@n14)
   B=Multiply(
    A=@n21
    B=MaterialFunctionCall(Mask In=@n23, Amount=@n24.R, Range=@n24.R, Start Offset=@n24.R, Sharpness=@n24.R) {materialFunction=&MF_MaskShifting}.Result
   )
   Alpha=@n21
  )
  n26 = Multiply(A=Time(), B=ScalarParameter() {defaultValue=0.5, group=TimeTrotter, parameterName=EmissiveGradient.PanningSpeed})
  n27 = TextureCoordinate()
  n28 = TextureSampleParameter2D!(UVs=Frac(None=Add(A=@n25, B=@n26)), "DDX(UVs)"=DDX(Value=@n27), "DDY(UVs)"=DDY(Value=@n27), Apply View MipBias) {
   group=TimeTrotter
   mipValueMode=TMVM_Derivative
   parameterName=PanningColorGradient
   samplerSource=SSM_Wrap_WorldGroupSettings
   sortPriority=6
   texture=/Game/Global/Textures/Default/Masks_and_Gradients/3pxRGBImage
  }
  n29 = Desaturation(None=@n28.RGB, Fraction=ScalarParameter() {group=TimeTrotter, parameterName=PanningColorGradient.Desaturation})
  n30 = NamedRerouteUsage()
  NamedRerouteDeclaration(None=Multiply(A=Multiply(A=@n29, B=@n30), B=ScalarParameter() {defaultValue=1, group=TimeTrotter, parameterName=PanningColorGradient.Brightness})) {name=AdvancedEmissiveOutput, nodeColor=<rgba 1 0 0.2578 1>}
  NamedRerouteDeclaration(None=Multiply(A=@n8.R, B=SmoothStep(Value=@n8.G) {constMax=0.1, constMin=0.02})) {name=EmissiveMask, nodeColor=<rgba 1 0.2578 0 1>}
 ]
}
