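# Control Rig for the F_MED_TimeTrotter female player body that applies corrective bone adjustments to collar, belt, glute, foot and upper-arm bones.
# A shared DeformRig function runs first, then spherical pose readers on thighs, head and right upper arm, plus a head DeltaRotation swing/twist function, weight ModifyTransforms offsets.
# Helmet, crouch, rifle, pistol, targeting and parachute variables gate or scale corrections, some smoothed by AlphaInterp.
# Constraints blend outer-foot and collar bones toward leg and clavicle bones.
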ControlRigBlueprint /BRCosmetics/Characters/Player/Female/Medium/Bodies/F_MED_TimeTrotter/Meshes/F_MED_TimeTrotter_CtrlRig
graphs [
 Multiply_1_ContainedGraph {nodes [
   n1 = RigVMFunctionEntryNode()
   RigVMFunctionReturnNode(Result=&RigVMFunction_MathDoubleMul(B=@n1.C, A=&RigVMFunction_MathDoubleMul(A=@n1.A, B=@n1.B).Result).Result)
  ]}
 Multiply_3_ContainedGraph {nodes [
   n1 = RigVMFunctionEntryNode()
   Multiply = &RigVMFunction_MathDoubleMul(A=@n1.A, B=@n1.B)
   RigVMFunctionReturnNode(Result=&RigVMFunction_MathDoubleMul(B=@n1.C, A=@Multiply.Result).Result)
  ]}
 Multiply_4_ContainedGraph {nodes [
   n1 = RigVMFunctionEntryNode()
   RigVMFunctionReturnNode(Result=&RigVMFunction_MathDoubleMul(B=@n1.C, A=&RigVMFunction_MathDoubleMul(A=@n1.A, B=@n1.B).Result).Result)
  ]}
 New Function_ContainedGraph {nodes [
   n1 = RigVMFunctionEntryNode()
   RigVMFunction_MathBoolToInteger = &RigVMFunction_MathBoolToInteger(Value=@n1.onlyTwist)
   RigVMFunction_MathQuaternionSwingTwist = &RigVMFunction_MathQuaternionSwingTwist(
    twistAxis=@n1.twistAxis
    Input=&RigUnit_GetRelativeTransformForItem(bChildInitial=False, bParentInitial=False, child=@n1.child, parent=@n1.parent).RelativeTransform.Rotation
   )
   n2 = RigVMDispatchNode(
    Values=((),())
    Index=@RigVMFunction_MathBoolToInteger.Result
    Values.0=@RigVMFunction_MathQuaternionSwingTwist.Swing
    Values.1=@RigVMFunction_MathQuaternionSwingTwist.Twist
   ) {template="DISPATCH_RigVMDispatch_SelectInt32(in Index,in Values,out Result)"}
   RigVMFunction_MathQuaternionToEuler = &RigVMFunction_MathQuaternionToEuler(Value=@n2.Result, RotationOrder=@n1.Rotation Order)
   > DISPATCH_RigVMDispatch_Print = RigVMDispatchNode(ScreenColor=(R=1,G=1,B=1,A=1), ScreenDuration=1.5, Enabled=@n1.debugRotation, Value=@RigVMFunction_MathQuaternionToEuler.Result) {template="DISPATCH_RigVMDispatch_Print(in Prefix,in Value,in Enabled,in ScreenDuration,in ScreenColor)"}
   Cast = RigVMDispatchNode(Value=@n1.triggerAxis) {template="DISPATCH_RigVMDispatch_CastEnumToInt(in Value,out Result)"}
   DISPATCH_RigVMDispatch_SelectInt32 = RigVMDispatchNode(
    Values=(0,0,0)
    Index=@Cast.Result
    Values.0=@RigVMFunction_MathQuaternionToEuler.Result.X
    Values.1=@RigVMFunction_MathQuaternionToEuler.Result.Y
    Values.2=@RigVMFunction_MathQuaternionToEuler.Result.Z
   ) {template="DISPATCH_RigVMDispatch_SelectInt32(in Index,in Values,out Result)"}
   Remap = &RigVMFunction_MathDoubleRemap(
    bClamp=False
    Value=@DISPATCH_RigVMDispatch_SelectInt32.Result
    TargetMinimum=@n1.outputMinMax.X
    TargetMaximum=@n1.outputMinMax.Y
    SourceMinimum=@n1.rotationMinMax.X
    SourceMaximum=@n1.rotationMinMax.Y
   )
   Absolute = &RigVMFunction_MathDoubleAbs(Value=@Remap.Result)
   Multiply = &RigVMFunction_MathDoubleMul(
    A=&RigVMFunction_MathDoubleClamp(Value=RigVMDispatchNode(True=@Absolute.Result, Condition=@n1.absoluteValue, False=@Remap.Result) {template="DISPATCH_RigVMDispatch_If(in Condition,in True,in False,out Result)"}.Result).Result
    B=@n1.premult
   )
   > RigVMFunctionReturnNode(outputQuat=@n2.Result, outputWeight=@Multiply.Result)
  ]}
 RigVMFunctionLibrary {nodes [DeltaRotation v4 = RigVMCollapseNode(
    ExecuteContext=()
    Rotation Order=ZYX
    child=(Type=Bone,Name="None")
    outputMinMax=(X=0,Y=1)
    parent=(Type=Bone,Name="None")
    premult=1
    rotationMinMax=(X=0,Y=0)
    twistAxis=(X=1,Y=0,Z=0)
   )]}
 RigVMFunction_Sequence_1_ContainedGraph {nodes [
   n1 = RigVMFunctionEntryNode()
   RigVMFunction_Sequence = &RigVMFunction_Sequence(ExecuteContext=@n1.ExecuteContext)
   n2 = &RigVMFunction_Sequence(ExecuteContext=@RigVMFunction_Sequence.B)
   n3 = &RigVMFunction_Sequence(ExecuteContext=@n2.B)
   n4 = &RigVMFunction_Sequence(ExecuteContext=@n3.B)
   n5 = &RigVMFunction_Sequence(ExecuteContext=@n4.B)
   n6 = &RigVMFunction_Sequence(ExecuteContext=@n5.B)
   n7 = &RigVMFunction_Sequence(ExecuteContext=@n6.B)
   RigVMFunctionReturnNode(A=@RigVMFunction_Sequence.A, B=@n2.A, C=@n3.A, D=@n4.A, E=@n5.A, F=@n6.A, G=@n7.A, H=@n7.B)
  ]}
 RigVMModel {nodes [
   RigUnit_BeginExecution = &RigUnit_BeginExecution()
   RigVMFunction_Sequence = RigVMAggregateNode(ExecuteContext=@RigUnit_BeginExecution.ExecutePin)
   FN_DeformRig_FNC = RigVMFunctionReferenceNode(
    DeformRig_ Curve="(deformRig_l_upperarm_up_alpha_72_#1=False,deformRig_r_upperarm_up_alpha_73_#2=False,deformRig_l_upperarm_fr_alpha_74_#3=False,deformRig_r_upperarm_fr_alpha_75_CB51B8 ...(truncated)"
    DeformRig_Bool="(upperarm_1_#4=True,elbow_3_#5=True,wrist_7_#6=True,groin_8_#7=True,glute_17_#8=True,knee_9_40C408324CA261 ...(truncated)"
    DeformRig_Debug=(upperarm_debug_18_#4=False,wrist_debug_20_#6=False,lat_debug_26_#9=False,groinGlute_debug_27_#7=False)
    DeformRig_Value="(upperarm_up_18_#10=(Rotation=(X=0,Y=0,Z=0,W=1),Translation=(X=8,Y=0,Z=2),Scale3D=(X=0.7,Y=1,Z=1)),upperarm_fwd_fr_19_#11=(Rotation=(X=0,Y=0,Z=0,W=1),Translation=(X=0,Y=0,Z=0),Scale3D=(X=0.5,Y=1,Z=1 ...(truncated)"
    ExecuteContext=@RigVMFunction_Sequence.A
   )
   > n1 = &RigUnit_SphericalPoseReader(
    ActiveRegionScaleFactors=(PositiveWidth=1,NegativeWidth=0.1,PositiveHeight=1,NegativeHeight=1)
    ActiveRegionSize=0.3
    Debug=(bDrawDebug=false,bDraw2D=false,bDrawLocalAxes=false,DebugScale=30,DebugSegments=30,DebugThickness=0.2)
    DriverAxis=(X=-1,Y=0,Z=0)
    DriverItem=(Type=Bone,Name="thigh_l")
    FalloffRegionScaleFactors=(PositiveWidth=1,NegativeWidth=0.5,PositiveHeight=1,NegativeHeight=1)
    FalloffSize=0.5
    OptionalParentItem=(Type=Bone,Name="pelvis")
    RotationOffset=(X=90,Y=0,Z=0)
   )
   n2 = &RigVMFunction_AlphaInterp(
    InterpSpeedDecreasing=5
    InterpSpeedIncreasing=5
    Scale=1
    bClampResult=False
    Value=&RigVMFunction_MathBoolToFloat(Value=RigVMVariableNode() {variable=OnlyCrouch}.Value).Result
   )
   > &RigUnit_ModifyTransforms(
    ItemToModify=((Item=(Type=Bone,Name="deform_glute_l"),Transform=(Rotation=(X=0,Y=0,Z=0,W=1),Translation=(X=-4,Y=0,Z=0),Scale3D=(X=1,Y=1,Z=1))))
    WeightMinimum=0
    Weight=&RigVMFunction_MathDoubleMul(B=@n2.Result, A=@n1.OutputParam).Result
   )
   &RigUnit_ModifyTransforms(
    ItemToModify=((Item=(Type=Bone,Name="thigh_l"),Transform=(Rotation=(X=0,Y=0,Z=0,W=1),Translation=(X=0,Y=0,Z=0),Scale3D=(X=1,Y=1,Z=1))))
    Weight=1
    WeightMinimum=0
    ExecutePin=@RigVMFunction_Sequence.B
   )
   > n3 = &RigUnit_SphericalPoseReader(
    ActiveRegionScaleFactors=(PositiveWidth=1,NegativeWidth=1,PositiveHeight=1,NegativeHeight=1)
    ActiveRegionSize=0.1
    Debug=(bDrawDebug=false,bDraw2D=false,bDrawLocalAxes=False,DebugScale=25,DebugSegments=20,DebugThickness=0.25)
    DriverAxis=(X=-1,Y=0,Z=0)
    DriverItem=(Type=Bone,Name="thigh_l")
    FalloffRegionScaleFactors=(PositiveWidth=1.5,NegativeWidth=0.1,PositiveHeight=1.5,NegativeHeight=1.5)
    FalloffSize=0.7
    FlipHeightScaling=true
    FlipWidthScaling=true
    OptionalParentItem=(Type=Bone,Name="None")
    RotationOffset=(X=106.2568,Y=27.914,Z=-53.0439)
   )
   > &RigUnit_ModifyTransforms(
    ItemToModify="((Item=(Type=Bone,Name=\"dyn_belt\"),Transform=(Rotation=(X=0.2172,Y=0.0302,Z=-0.1342,W=0.9664),Translation=(X=-5,Y=-1.3528,Z=1.7479),Scale3D=(X=1,Y=1,Z=1))),(Item=(Type=Bone,Name=\"dyn_belt_01\"),Transform=(Rotation=(X=0,Y=0,Z=0,W=1),Translation=(X=-1,Y=0,Z=0 ...(truncated)"
    WeightMinimum=0
    Weight=@n3.OutputParam
   )
   n4 = &RigVMFunction_ControlFlowBranch(ExecuteContext=@RigVMFunction_Sequence.C, Condition=RigVMVariableNode() {variable=HelmetOn}.Value)
   DeltaRotation v4_7 = RigVMFunctionReferenceNode(
    Rotation Order=ZYX
    child=(Type=Bone,Name="head")
    outputMinMax=(X=0,Y=1)
    parent=(Type=Null,Name="head_null")
    premult=1
    rotationMinMax=(X=-30,Y=-60)
    triggerAxis=Z
    twistAxis=(X=1,Y=0,Z=0)
    ExecuteContext=@n4.True
   )
   > &RigUnit_ModifyTransforms(
    ItemToModify="((Item=(Type=Bone,Name=\"base_Collar_b\"),Transform=(Rotation=(X=0,Y=-0.4226,Z=0,W=0.9063),Translation=(X=0,Y=0,Z=0),Scale3D=(X=0.6,Y=1,Z=1))),(Item=(Type=Bone,Name=\"base_Collar_m_r\"),Transform=(Rotation=(X=0,Y=-0.1736,Z=0,W=0.9848),Translation=(X=0,Y=0,Z=0) ...(truncated)"
    WeightMinimum=0
    Weight=@"DeltaRotation v4_7".outputWeight
   )
   > n5 = &RigUnit_SphericalPoseReader(
    ActiveRegionScaleFactors=(PositiveWidth=1.5,NegativeWidth=1.5,PositiveHeight=1,NegativeHeight=1)
    ActiveRegionSize=0.05
    Debug=(bDrawDebug=true,bDraw2D=False,bDrawLocalAxes=False,DebugScale=25,DebugSegments=20,DebugThickness=0.25)
    DriverAxis=(X=1,Y=0,Z=0)
    DriverItem=(Type=Bone,Name="head")
    FalloffRegionScaleFactors=(PositiveWidth=1.5,NegativeWidth=1.5,PositiveHeight=1,NegativeHeight=1)
    FalloffSize=0.15
    FlipHeightScaling=False
    FlipWidthScaling=False
    OptionalParentItem=(Type=Bone,Name="spine_05")
    RotationOffset=(X=-45,Y=-3,Z=65)
   )
   Multiply_6 = RigVMAggregateNode(A=@n5.OutputParam, B=RigVMVariableNode() {variable=Riffle}.Value)
   > &RigUnit_ModifyTransforms(
    ItemToModify="((Item=(Type=Bone,Name=\"dyn_Collar_m_r\"),Transform=(Rotation=(X=0,Y=0,Z=0,W=1),Translation=(X=0.05,Y=0,Z=0),Scale3D=(X=0.5,Y=1,Z=1))),(Item=(Type=Bone,Name=\"dyn_Collar_b\"),Transform=(Rotation=(X=0,Y=0,Z=0,W=1),Translation=(X=0,Y=0,Z=0),Scale3D=(X=0.8,Y=1,Z ...(truncated)"
    WeightMinimum=0
    Weight=@Multiply_6.Result
   )
   > n6 = &RigUnit_SphericalPoseReader(
    ActiveRegionScaleFactors=(PositiveWidth=1,NegativeWidth=1,PositiveHeight=1,NegativeHeight=1)
    ActiveRegionSize=0.1
    Debug=(bDrawDebug=false,bDraw2D=False,bDrawLocalAxes=False,DebugScale=25,DebugSegments=20,DebugThickness=0.25)
    DriverAxis=(X=1,Y=0,Z=0)
    DriverItem=(Type=Bone,Name="head")
    FalloffRegionScaleFactors=(PositiveWidth=1,NegativeWidth=1,PositiveHeight=1,NegativeHeight=1)
    FalloffSize=0.3
    FlipHeightScaling=False
    FlipWidthScaling=False
    OptionalParentItem=(Type=Bone,Name="spine_05")
    RotationOffset=(X=-40,Y=40,Z=50)
   )
   Multiply_18 = RigVMAggregateNode(A=@n6.OutputParam, B=RigVMVariableNode() {variable=Pistol}.Value, C=RigVMVariableNode() {variable=Targeting}.Value)
   n7 = &RigVMFunction_AlphaInterp(InterpSpeedDecreasing=10, InterpSpeedIncreasing=10, Scale=1, bClampResult=False, Value=@Multiply_18.Result)
   > n8 = &RigUnit_ModifyTransforms(
    ItemToModify="((Item=(Type=Bone,Name=\"dyn_Collar_m_r\"),Transform=(Rotation=(X=-0.4082,Y=-0.2346,Z=-0.1094,W=0.8754),Translation=(X=0,Y=0,Z=0),Scale3D=(X=0.5,Y=1,Z=1))),(Item=(Type=Bone,Name=\"dyn_Collar_b\"),Transform=(Rotation=(X=0,Y=0,Z=0,W=1),Translation=(X=0,Y=0,Z=0), ...(truncated)"
    WeightMinimum=0
    Weight=@n7.Result
   )
   n9 = &RigVMFunction_ControlFlowBranch(ExecuteContext=@n8.ExecutePin, Condition=RigVMVariableNode() {variable=HelmetOn}.Value)
   DeltaRotation v4_2 = RigVMFunctionReferenceNode(
    Rotation Order=ZYX
    child=(Type=Bone,Name="head")
    outputMinMax=(X=0,Y=1)
    parent=(Type=Null,Name="head_null")
    premult=1
    rotationMinMax=(X=-20,Y=-50)
    triggerAxis=Z
    twistAxis=(X=1,Y=0,Z=0)
    ExecuteContext=@n9.True
   )
   > &RigUnit_ModifyTransforms(
    ItemToModify="((Item=(Type=Bone,Name=\"base_Collar_b\"),Transform=(Rotation=(X=0.0014,Y=-0.1316,Z=0.0103,W=0.9913),Translation=(X=0,Y=0,Z=0),Scale3D=(X=0.5,Y=1,Z=1))),(Item=(Type=Bone,Name=\"base_Collar_m_r\"),Transform=(Rotation=(X=0,Y=0.1356,Z=0,W=-0.9908),Translation=(X= ...(truncated)"
    WeightMinimum=0
    Weight=&RigVMFunction_MathDoubleMul(A=@"DeltaRotation v4_2".outputWeight, B=RigVMVariableNode() {variable=Crouch walk}.Value).Result
   )
   n10 = &RigVMFunction_ControlFlowBranch(ExecuteContext=@RigVMFunction_Sequence.D, Condition=RigVMVariableNode() {variable=HelmetOn}.Value)
   n11 = &RigUnit_SphericalPoseReader(
    ActiveRegionScaleFactors=(PositiveWidth=1,NegativeWidth=0.5,PositiveHeight=1,NegativeHeight=1)
    ActiveRegionSize=0.1
    Debug=(bDrawDebug=false,DebugScale=25,DebugSegments=20,DebugThickness=0.25)
    DriverAxis=(X=0,Y=-1,Z=0)
    DriverItem=(Type=Bone,Name="head")
    FalloffRegionScaleFactors=(PositiveWidth=1,NegativeWidth=0.5,PositiveHeight=1,NegativeHeight=1)
    FalloffSize=0.25
    OptionalParentItem=(Type=Bone,Name="neck_02")
    RotationOffset=(X=-46.5047,Y=-4.162,Z=4.0076)
    ExecutePin=@n10.True
   )
   > &RigUnit_ModifyTransforms(
    ItemToModify="((Item=(Type=Bone,Name=\"base_Collar_l\"),Transform=(Rotation=(X=0,Y=-0.1736,Z=0,W=0.9848),Translation=(X=0,Y=0,Z=0),Scale3D=(X=0.6,Y=1,Z=1))),(Item=(Type=Bone,Name=\"base_Collar_m_l\"),Transform=(Rotation=(X=-0.0113,Y=-0.2586,Z=-0.0421,W=0.965),Translation=(X ...(truncated)"
    WeightMinimum=0
    Weight=@n11.OutputParam
   )
   &RigUnit_ModifyTransforms(
    ItemToModify=((Item=(Type=Bone,Name="upperarm_r"),Transform=(Rotation=(X=0,Y=0,Z=0,W=1),Translation=(X=0,Y=0,Z=0),Scale3D=(X=1,Y=1,Z=1))))
    Weight=1
    WeightMinimum=0
    ExecutePin=@RigVMFunction_Sequence.E
   )
   > &RigUnit_RotationConstraintLocalSpaceOffset(
    AdvancedSettings=(InterpolationType=Shortest,RotationOrderForFilter=XZY)
    Child=(Type=Bone,Name="deform_foot_out_base_l")
    Filter=(bX=True,bY=True,bZ=True)
    Parents=((Item=(Type=Bone,Name="foot_l"),Weight=0.5),(Item=(Type=Bone,Name="calf_twist_02_l"),Weight=0.5))
    Weight=1
   )
   > &RigUnit_RotationConstraintLocalSpaceOffset(
    AdvancedSettings=(InterpolationType=Shortest,RotationOrderForFilter=XZY)
    Child=(Type=Bone,Name="deform_foot_out_base_r")
    Filter=(bX=True,bY=True,bZ=True)
    Parents=((Item=(Type=Bone,Name="foot_r"),Weight=0.5),(Item=(Type=Bone,Name="calf_twist_02_r"),Weight=0.5))
    Weight=1
   )
   > &RigUnit_ParentConstraint(
    AdvancedSettings=(InterpolationType=Average,RotationOrderForFilter=XZY)
    Child=(Type=Bone,Name="deform_foot_out_base_r")
    Filter=(TranslationFilter=(bX=True,bY=True,bZ=True),RotationFilter=(bX=false,bY=false,bZ=false),ScaleFilter=(bX=True,bY=True,bZ=True))
    Parents=((Item=(Type=Bone,Name="calf_r"),Weight=0.5),(Item=(Type=Bone,Name="foot_r"),Weight=0.5))
    Weight=1
   )
   > n12 = &RigUnit_ParentConstraint(
    AdvancedSettings=(InterpolationType=Average,RotationOrderForFilter=XZY)
    Child=(Type=Bone,Name="deform_foot_out_base_l")
    Filter=(TranslationFilter=(bX=True,bY=True,bZ=True),RotationFilter=(bX=false,bY=false,bZ=false),ScaleFilter=(bX=True,bY=True,bZ=True))
    Parents=((Item=(Type=Bone,Name="foot_l"),Weight=0.5),(Item=(Type=Bone,Name="calf_l"),Weight=0.5))
    Weight=1
   )
   n13 = &RigVMFunction_ControlFlowBranch(ExecuteContext=@n12.ExecutePin, Condition=RigVMVariableNode() {variable=HelmetOn}.Value)
   n14 = &RigUnit_SphericalPoseReader(
    ActiveRegionScaleFactors=(PositiveWidth=1,NegativeWidth=1,PositiveHeight=1,NegativeHeight=1)
    ActiveRegionSize=0.1
    Debug=(bDrawDebug=false,bDraw2D=False,bDrawLocalAxes=False,DebugScale=25,DebugSegments=20,DebugThickness=0.25)
    DriverAxis=(X=1,Y=0,Z=0)
    DriverItem=(Type=Bone,Name="head")
    FalloffRegionScaleFactors=(PositiveWidth=1,NegativeWidth=1,PositiveHeight=1,NegativeHeight=1)
    FalloffSize=0.4
    FlipHeightScaling=False
    FlipWidthScaling=False
    OptionalParentItem=(Type=Bone,Name="spine_05")
    RotationOffset=(X=-55,Y=40,Z=50)
    ExecutePin=@n13.False
   )
   Multiply = RigVMAggregateNode(A=@n14.OutputParam, B=RigVMVariableNode() {variable=Pistol}.Value, C=RigVMVariableNode() {variable=Targeting}.Value)
   AlphaInterp = &RigVMFunction_AlphaInterp(InterpSpeedDecreasing=10, InterpSpeedIncreasing=10, Scale=1, bClampResult=False, Value=@Multiply.Result)
   > &RigUnit_ModifyTransforms(
    ItemToModify=((Item=(Type=Bone,Name="dyn_Collar_m_l"),Transform=(Rotation=(X=0,Y=0,Z=0.1802,W=-0.9836),Translation=(X=0,Y=0,Z=0),Scale3D=(X=1,Y=1,Z=1))))
    WeightMinimum=0
    Weight=@AlphaInterp.Result
   )
   n15 = &RigUnit_SphericalPoseReader(
    ActiveRegionScaleFactors=(PositiveWidth=1,NegativeWidth=1,PositiveHeight=1,NegativeHeight=1)
    ActiveRegionSize=0.03
    Debug=(bDrawDebug=false,bDraw2D=False,bDrawLocalAxes=False,DebugScale=25,DebugSegments=20,DebugThickness=0.25)
    DriverAxis=(X=-1,Y=0,Z=0)
    DriverItem=(Type=Bone,Name="upperarm_r")
    FalloffRegionScaleFactors=(PositiveWidth=1,NegativeWidth=1,PositiveHeight=1,NegativeHeight=1)
    FalloffSize=0.2
    FlipHeightScaling=False
    FlipWidthScaling=False
    OptionalParentItem=(Type=Bone,Name="None")
    RotationOffset=(X=-84.7555,Y=23.9264,Z=38.419)
    ExecutePin=@RigVMFunction_Sequence.F
   )
   > &RigUnit_ModifyTransforms(
    ItemToModify="((Item=(Type=Bone,Name=\"base_Collar_r\"),Transform=(Rotation=(X=-0.0124,Y=0.4226,Z=0.0058,W=0.9062),Translation=(X=0,Y=0,Z=-2.2708),Scale3D=(X=1,Y=1,Z=1))),(Item=(Type=Bone,Name=\"base_Collar_m_r\"),Transform=(Rotation=(X=0,Y=0,Z=0,W=1),Translation=(X=0,Y=0,Z ...(truncated)"
    WeightMinimum=0
    Weight=@n15.OutputParam
   )
   > &RigUnit_ModifyTransforms(
    ItemToModify="((Item=(Type=Bone,Name=\"base_Collar_m_l\"),Transform=(Rotation=(X=0,Y=0,Z=0,W=1),Translation=(X=0,Y=2,Z=0),Scale3D=(X=1,Y=1,Z=1))),(Item=(Type=Bone,Name=\"base_Collar_l\"),Transform=(Rotation=(X=0,Y=0,Z=0,W=1),Translation=(X=0,Y=0,Z=0),Scale3D=(X=1,Y=1,Z=1))) ...(truncated)"
    Weight=1
    WeightMinimum=0
   )
   > &RigUnit_ParentConstraint(
    AdvancedSettings=(InterpolationType=Average,RotationOrderForFilter=XZY)
    Child=(Type=Bone,Name="base_Collar_l")
    Filter=(TranslationFilter=(bX=true,bY=true,bZ=true),RotationFilter=(bX=True,bY=True,bZ=True),ScaleFilter=(bX=True,bY=True,bZ=True))
    Parents=((Item=(Type=Bone,Name="clavicle_l"),Weight=1))
    Weight=0.1
   )
   > &RigUnit_ParentConstraint(
    AdvancedSettings=(InterpolationType=Average,RotationOrderForFilter=XZY)
    Child=(Type=Bone,Name="base_Collar_m_l")
    Filter=(TranslationFilter=(bX=false,bY=false,bZ=false),RotationFilter=(bX=true,bY=true,bZ=true),ScaleFilter=(bX=True,bY=True,bZ=True))
    Parents=((Item=(Type=Bone,Name="clavicle_l"),Weight=1))
    Weight=0.6
   )
   > &RigUnit_ParentConstraint(
    AdvancedSettings=(InterpolationType=Average,RotationOrderForFilter=XZY)
    Child=(Type=Bone,Name="base_Collar_r")
    Filter=(TranslationFilter=(bX=true,bY=true,bZ=true),RotationFilter=(bX=True,bY=True,bZ=True),ScaleFilter=(bX=True,bY=True,bZ=True))
    Parents=((Item=(Type=Bone,Name="clavicle_r"),Weight=1))
    Weight=0.3
   )
   > &RigUnit_ParentConstraint(
    AdvancedSettings=(InterpolationType=Average,RotationOrderForFilter=XZY)
    Child=(Type=Bone,Name="base_Collar_m_r")
    Filter=(TranslationFilter=(bX=false,bY=false,bZ=false),RotationFilter=(bX=true,bY=true,bZ=true),ScaleFilter=(bX=True,bY=True,bZ=True))
    Parents=((Item=(Type=Bone,Name="clavicle_r"),Weight=1))
    Weight=0.6
   )
   RigVMFunction_MathBoolToFloat = &RigVMFunction_MathBoolToFloat(Value=RigVMVariableNode() {variable=parachute open}.Value)
   &RigUnit_ModifyTransforms(
    ItemToModify="((Item=(Type=Bone,Name=\"deform_upperarm_up_base_l\"),Transform=(Rotation=(X=0,Y=0,Z=0,W=-1),Translation=(X=2.2471,Y=0,Z=0),Scale3D=(X=1,Y=1,Z=1))),(Item=(Type=Bone,Name=\"deform_upperarm_up_base_r\"),Transform=(Rotation=(X=0,Y=0,Z=0,W=1),Translation=(X=-2,Y=0 ...(truncated)"
    WeightMinimum=0
    Weight=@RigVMFunction_MathBoolToFloat.Result
    ExecutePin=@RigVMFunction_Sequence.G
   )
   n16 = &RigVMFunction_MathDoubleMul(A=RigVMVariableNode() {variable=Pistol}.Value, B=RigVMVariableNode() {variable=Targeting}.Value)
   &RigUnit_ModifyTransforms(
    ItemToModify=((Item=(Type=Bone,Name="base_upperarm"),Transform=(Rotation=(X=-0.3359,Y=0.8996,Z=-0.2048,W=-0.1897),Translation=(X=-4.9598,Y=0.5997,Z=-1.7993),Scale3D=(X=1,Y=1,Z=1))))
    WeightMinimum=0
    Weight=@n16.Result
    ExecutePin=@RigVMFunction_Sequence.H
   )
   &RigVMFunction_AlphaInterp(InterpSpeedDecreasing=50, InterpSpeedIncreasing=50, Scale=1, bClampResult=False, Value=@n16.Result)
   RigVMVariableNode() {variable=Targeting}
  ]}
]
hierarchy [
 root {loc=<0 0 0>, rot=<0 0 0 1>}
 pelvis {parent=root, loc=<0 -0.5301 98.0571>, rot=<0.0127 0.707 -0.0127 -0.707>}
 spine_01 {parent=pelvis, loc=<6.1208 -0.9328 0>, rot=<0 0 0.034 -0.9994>}
 spine_02 {parent=spine_01, loc=<6.5997 -0.2095 0>, rot=<0 0 0.0238 -0.9997>}
 spine_03 {parent=spine_02, loc=<6.4914 -0.1237 0>, rot=<0 0 0.0149 -0.9999>}
 spine_04 {parent=spine_03, loc=<6.9598 0.0794 0>, rot=<0 0 -0.0132 -0.9999>}
 spine_05 {parent=spine_04, loc=<6.9987 0 0>, rot=<0 0 -0.0015 -1>}
 neck_01 {parent=spine_05, loc=<11.3138 0.1529 0>, rot=<0 0 -0.2003 -0.9797>}
 neck_02 {parent=neck_01, loc=<4.5627 0 0>, rot=<0 0 0.0504 -0.9987>}
 head {parent=neck_02, loc=<5.3914 0 0>, rot=<0 0 0.115 -0.9934>}
 base_ponytail {parent=head, loc=<11.6238 -7.3448 0>, rot=<0.5974 -0.3783 0.3783 -0.5974>}
 dyn_ponytail_01 {parent=base_ponytail, loc=<0 0 0>, rot=<0 0 0 1>}
 base_earring_l {parent=head, loc=<-0.897 0.1238 -6.5497>, rot=<0.0011 -0.9596 -0.2815 0.0017>}
 dyn_earring_l {parent=base_earring_l, loc=<0 0 0>, rot=<0 0 0 1>}
 base_earring_r {parent=head, loc=<-0.8973 0.1238 6.5497>, rot=<-0.0018 0.2989 0.9543 -0.0011>}
 dyn_earring_r {parent=base_earring_r, loc=<0 0 0>, rot=<0 0 0 1>}
 base_hair_a {parent=head, loc=<8.2703 1.9152 -7.9026>, rot=<0.4301 -0.9 -0.0501 0.0502>}
 dyn_hair_a {parent=base_hair_a, loc=<0 0 0>, rot=<0 0 0 1>}
 base_hair_b {parent=head, loc=<4.5876 2.6935 -7.7173>, rot=<-0.0107 -0.9989 -0.0303 -0.0355>}
 dyn_hair_b {parent=base_hair_b, loc=<0 0 0>, rot=<0 0 0 1>}
 base_hair_c {parent=head, loc=<8.6069 6.1598 -6.698>, rot=<0.1473 -0.9768 -0.1349 -0.0776>}
 dyn_hair_c {parent=base_hair_c, loc=<0 0 0>, rot=<0 0 0 1>}
 dyn_hair_c_01 {parent=dyn_hair_c, loc=<3.4692 0 0>, rot=<0 0.1439 0 -0.9896>}
 base_hair_d {parent=head, loc=<8.5124 8.8177 -4.2923>, rot=<0.0037 -0.8837 -0.4681 0.0019>}
 dyn_hair_d {parent=base_hair_d, loc=<0 0 0>, rot=<0 0 0 1>}
 base_hair_e {parent=head, loc=<8.1895 10.2108 -0.0538>, rot=<-0.0029 0.7071 0.7071 -0.0029>}
 dyn_hair_e {parent=base_hair_e, loc=<0 0 0>, rot=<0 0 0 1>}
 base_hair_f {parent=head, loc=<8.3142 10.2118 3.2559>, rot=<0.0131 0.6999 0.7139 -0.0186>}
 dyn_hair_f {parent=base_hair_f, loc=<0 0 0>, rot=<0 0 0 1>}
 base_hair_h {parent=head, loc=<7.2788 7.8428 5.8061>, rot=<0.0149 0.2796 0.96 -0.0026>}
 dyn_hair_h {parent=base_hair_h, loc=<0 0 0>, rot=<0 0 0 1>}
 base_hair_i {parent=head, loc=<5.1181 -2.5592 8.0966>, rot=<0.0406 0.0793 0.9508 -0.2967>}
 dyn_hair_i {parent=base_hair_i, loc=<0 0 0>, rot=<0 0 0 1>}
 base_hair_j {parent=head, loc=<11.0449 4.3669 -7.7608>, rot=<0.3162 -0.5099 0.0959 -0.7943>}
 dyn_hair_j {parent=base_hair_j, loc=<0 0 0>, rot=<0 0 0 1>}
 base_hair_k {parent=head, loc=<10.3452 5.7279 -7.158>, rot=<0.2365 -0.737 0.081 -0.628>}
 dyn_hair_k {parent=base_hair_k, loc=<0 0 0>, rot=<0 0 0 1>}
 base_hair_l {parent=head, loc=<7.1628 3.3943 7.281>, rot=<0.0086 0.1145 0.9934 -0.0076>}
 dyn_hair_l {parent=base_hair_l, loc=<0 0 0>, rot=<0 0 0 1>}
 clavicle_l {parent=spine_05, loc=<5.4831 -0.1053 -3.2946>, rot=<0.0465 -0.7289 -0.009 -0.6829>}
 upperarm_l {parent=clavicle_l, loc=<9.8658 -1.2136 -0.4105>, rot=<0.0164 -0.3245 -0.0269 -0.9454>}
 lowerarm_l {parent=upperarm_l, loc=<26.465 -0.8345 0.259>, rot=<0.0036 0.0161 -0.2298 -0.9731>}
 hand_l {parent=lowerarm_l, loc=<22.7355 -0.3203 -0.6105>, rot=<-0.6784 -0.0223 0.0735 -0.7307>}
 thumb_01_l {parent=hand_l, loc=<1.3118 -0.3191 -2.459>, rot=<0.6546 -0.4042 0.0664 -0.6354>}
 thumb_02_l {parent=thumb_01_l, loc=<5.9336 -0.0448 0.0401>, rot=<-0.0037 0.0213 0.298 -0.9543>}
 thumb_03_l {parent=thumb_02_l, loc=<2.2223 0.3682 -0.0907>, rot=<0.0118 -0.013 0.0066 -0.9998>}
 index_metacarpal_l {parent=hand_l, loc=<1.6535 0.0183 -2.96>, rot=<0.0158 -0.0124 -0.0173 -0.9996>}
 index_01_l {parent=index_metacarpal_l, loc=<8.3008 0.0357 -0.0524>, rot=<-0.0044 0.0485 0.0924 -0.9945>}
 index_02_l {parent=index_01_l, loc=<3.3587 -0.3004 0.0452>, rot=<-0.002 0.0027 0.3225 -0.9466>}
 index_03_l {parent=index_02_l, loc=<2.5578 -0.123 -0.0252>, rot=<-0.002 0.0067 0.2558 -0.9667>}
 middle_metacarpal_l {parent=hand_l, loc=<2.1856 -0.1725 -0.5984>, rot=<0.006 0.0168 -0.0329 -0.9993>}
 middle_01_l {parent=middle_metacarpal_l, loc=<7.3275 -0.2447 0.0129>, rot=<-0.0208 0.0042 0.0858 -0.9961>}
 middle_02_l {parent=middle_01_l, loc=<4.0123 0.1401 0.1175>, rot=<-0.0218 0.0258 0.3512 -0.9357>}
 middle_03_l {parent=middle_02_l, loc=<2.6077 0.0534 0.1373>, rot=<0.0082 0.0188 0.1708 -0.9851>}
 ring_metacarpal_l {parent=hand_l, loc=<2.2074 -0.3625 1.3638>, rot=<-0.0284 0.0722 -0.0458 -0.9959>}
 ring_01_l {parent=ring_metacarpal_l, loc=<7.0295 -0.6197 -0.1446>, rot=<-0.0489 -0.0252 0.1892 -0.9804>}
 ring_02_l {parent=ring_01_l, loc=<3.6174 0.4647 -0.1027>, rot=<0 0 0.2806 -0.9598>}
 ring_03_l {parent=ring_02_l, loc=<2.4152 0.114 0.0199>, rot=<0 0 0.1189 -0.9929>}
 pinky_metacarpal_l {parent=hand_l, loc=<1.7616 -0.9753 3.3814>, rot=<-0.0397 0.0814 -0.069 -0.9935>}
 pinky_01_l {parent=pinky_metacarpal_l, loc=<7.1709 -0.4831 -0.1094>, rot=<-0.0654 -0.0035 0.2242 -0.9723>}
 pinky_02_l {parent=pinky_01_l, loc=<2.8274 0.0893 -0.1043>, rot=<0.0037 0.0011 0.2854 -0.9584>}
 pinky_03_l {parent=pinky_02_l, loc=<2.2046 0.0307 -0.0078>, rot=<-0.0002 0.0098 0.1445 -0.9895>}
 deform_wrist_dn_base_l {parent=hand_l, loc=<0 -1.5326 0>, rot=<0 0 0 -1>}
 deform_wrist_dn_l {parent=deform_wrist_dn_base_l, loc=<0 0 0>, rot=<0 0 0 -1>}
 deform_wrist_up_base_l {parent=hand_l, loc=<0 2.3009 0>, rot=<0 0 0 -1>}
 deform_wrist_up_l {parent=deform_wrist_up_base_l, loc=<0 0 0>, rot=<0 0 0 -1>}
 lowerarm_twist_02_l {parent=lowerarm_l, loc=<10.6707 0 0>, rot=<0 0 0 1>}
 lowerarm_twist_01_l {parent=lowerarm_l, loc=<18.316 0 0>, rot=<0 0 0 1>}
 deform_elbow_in_base_l {parent=lowerarm_l, loc=<0.1834 3.6062 -0.2572>, rot=<0 0 0 -1>}
 deform_elbow_in_l {parent=deform_elbow_in_base_l, loc=<0 0 0>, rot=<0 0 0 -1>}
 deform_elbow_out_base_l {parent=lowerarm_l, loc=<0 -4 0>, rot=<0 0 0 1>}
 deform_elbow_out_l {parent=deform_elbow_out_base_l, loc=<0 0 0>, rot=<0 0 0 1>}
 upperarm_twist_01_l {parent=upperarm_l, loc=<7.2135 -0.1134 0>, rot=<0 0 0 1>}
 deform_upperarm_bk_base_l {parent=upperarm_twist_01_l, loc=<-6.4053 -4.3746 0.314>, rot=<0 0 0 -1>}
 deform_upperarm_bk_l {parent=deform_upperarm_bk_base_l, loc=<0 0 0>, rot=<0 0 0 -1>}
 deform_upperarm_fr_base_l {parent=upperarm_twist_01_l, loc=<-5.6312 5.092 0.3354>, rot=<0 0 0 -1>}
 deform_upperarm_fr_l {parent=deform_upperarm_fr_base_l, loc=<0 0 0>, rot=<0 0 0 -1>}
 deform_upperarm_up_base_l {parent=upperarm_twist_01_l, loc=<-8.891 0.4825 3.9517>, rot=<0 0 0 -1>}
 deform_upperarm_up_l {parent=deform_upperarm_up_base_l, loc=<0 0 0>, rot=<0 0 0 -1>}
 upperarm_twist_02_l {parent=upperarm_l, loc=<19.7998 -0.1134 0>, rot=<0 0 0 1>}
 clavicle_r {parent=spine_05, loc=<5.4831 -0.1053 3.2946>, rot=<0.7289 0.0465 -0.6829 0.009>}
 upperarm_r {parent=clavicle_r, loc=<-9.8658 1.2136 0.4105>, rot=<0.0164 -0.3245 -0.0269 -0.9454>}
 lowerarm_r {parent=upperarm_r, loc=<-26.465 0.8345 -0.259>, rot=<0.0036 0.0161 -0.2298 -0.9731>}
 hand_r {parent=lowerarm_r, loc=<-22.7355 0.3203 0.6106>, rot=<-0.6784 -0.0223 0.0735 -0.7307>}
 thumb_01_r {parent=hand_r, loc=<-1.3118 0.3191 2.459>, rot=<0.6546 -0.4042 0.0664 -0.6354>}
 thumb_02_r {parent=thumb_01_r, loc=<-5.9336 0.0448 -0.0401>, rot=<-0.0037 0.0213 0.298 -0.9543>}
 thumb_03_r {parent=thumb_02_r, loc=<-2.2223 -0.3682 0.0907>, rot=<0.0118 -0.013 0.0066 -0.9998>}
 index_metacarpal_r {parent=hand_r, loc=<-1.6535 -0.0183 2.96>, rot=<0.0158 -0.0124 -0.0173 -0.9996>}
 index_01_r {parent=index_metacarpal_r, loc=<-8.3008 -0.0357 0.0524>, rot=<-0.0044 0.0485 0.0924 -0.9945>}
 index_02_r {parent=index_01_r, loc=<-3.3587 0.3004 -0.0452>, rot=<-0.002 0.0027 0.3225 -0.9466>}
 index_03_r {parent=index_02_r, loc=<-2.5578 0.123 0.0252>, rot=<-0.002 0.0067 0.2558 -0.9667>}
 middle_metacarpal_r {parent=hand_r, loc=<-2.1856 0.1725 0.5984>, rot=<0.006 0.0168 -0.0329 -0.9993>}
 middle_01_r {parent=middle_metacarpal_r, loc=<-7.3275 0.2447 -0.0129>, rot=<-0.0208 0.0042 0.0858 -0.9961>}
 middle_02_r {parent=middle_01_r, loc=<-4.0123 -0.1401 -0.1175>, rot=<-0.0218 0.0258 0.3512 -0.9357>}
 middle_03_r {parent=middle_02_r, loc=<-2.6077 -0.0534 -0.1373>, rot=<0.0082 0.0188 0.1708 -0.9851>}
 ring_metacarpal_r {parent=hand_r, loc=<-2.2074 0.3625 -1.3638>, rot=<-0.0284 0.0722 -0.0458 -0.9959>}
 ring_01_r {parent=ring_metacarpal_r, loc=<-7.0295 0.6197 0.1446>, rot=<-0.0489 -0.0252 0.1892 -0.9804>}
 ring_02_r {parent=ring_01_r, loc=<-3.6174 -0.4647 0.1027>, rot=<0 0 0.2806 -0.9598>}
 ring_03_r {parent=ring_02_r, loc=<-2.4152 -0.114 -0.0199>, rot=<0 0 0.1189 -0.9929>}
 pinky_metacarpal_r {parent=hand_r, loc=<-1.7616 0.9753 -3.3814>, rot=<-0.0397 0.0814 -0.069 -0.9935>}
 pinky_01_r {parent=pinky_metacarpal_r, loc=<-7.1709 0.4831 0.1094>, rot=<-0.0654 -0.0035 0.2242 -0.9723>}
 pinky_02_r {parent=pinky_01_r, loc=<-2.8274 -0.0893 0.1043>, rot=<0.0037 0.0011 0.2854 -0.9584>}
 pinky_03_r {parent=pinky_02_r, loc=<-2.2046 -0.0307 0.0078>, rot=<-0.0002 0.0098 0.1445 -0.9895>}
 deform_wrist_dn_base_r {parent=hand_r, loc=<-0.0001 1.5328 0>, rot=<0 0 0 -1>}
 deform_wrist_dn_r {parent=deform_wrist_dn_base_r, loc=<0 0 0>, rot=<0 0 0 -1>}
 deform_wrist_up_base_r {parent=hand_r, loc=<-0.0002 -2.3007 -0.0001>, rot=<0 0 0 -1>}
 deform_wrist_up_r {parent=deform_wrist_up_base_r, loc=<0 0 0>, rot=<0 0 0 -1>}
 lowerarm_twist_02_r {parent=lowerarm_r, loc=<-10.6707 0 0>, rot=<0 0 0 -1>}
 lowerarm_twist_01_r {parent=lowerarm_r, loc=<-18.316 0 0>, rot=<0 0 0 -1>}
 deform_elbow_in_base_r {parent=lowerarm_r, loc=<-0.1834 -3.6062 0.2572>, rot=<0 0 0 -1>}
 deform_elbow_in_r {parent=deform_elbow_in_base_r, loc=<0 0 0>, rot=<0 0 0 -1>}
 deform_elbow_out_base_r {parent=lowerarm_r, loc=<0 4 -0.0001>, rot=<0 0 0 -1>}
 deform_elbow_out_r {parent=deform_elbow_out_base_r, loc=<0 0 0>, rot=<0 0 0 1>}
 upperarm_twist_01_r {parent=upperarm_r, loc=<-7.2135 0.1134 0>, rot=<0 0 0 -1>}
 deform_upperarm_bk_base_r {parent=upperarm_twist_01_r, loc=<6.4053 4.3746 -0.314>, rot=<0 0 0 -1>}
 deform_upperarm_bk_r {parent=deform_upperarm_bk_base_r, loc=<0 0 0>, rot=<0 0 0 -1>}
 deform_upperarm_fr_base_r {parent=upperarm_twist_01_r, loc=<5.6312 -5.092 -0.3354>, rot=<0 0 0 -1>}
 deform_upperarm_fr_r {parent=deform_upperarm_fr_base_r, loc=<0 0 0>, rot=<0 0 0 -1>}
 base_upperarm {parent=deform_upperarm_fr_r, loc=<2.6355 -0.6881 1.4101>, rot=<0.1462 -0.8112 0.5177 0.2291>}
 deform_upperarm_up_base_r {parent=upperarm_twist_01_r, loc=<8.891 -0.4825 -3.9517>, rot=<0 0 0 -1>}
 deform_upperarm_up_r {parent=deform_upperarm_up_base_r, loc=<0 0 0>, rot=<0 0 0 -1>}
 upperarm_twist_02_r {parent=upperarm_r, loc=<-19.7998 0.1134 0>, rot=<0 0 0 -1>}
 attach_backpack {parent=spine_05, loc=<-0.123 -13.473 0>, rot=<-0.0093 -0.707 0.0093 -0.707>}
 Backpack_BR {parent=attach_backpack, type=Null, loc=<0 3.5 0>, rot=<0 0 0 1>}
 attach_cape {parent=spine_05, loc=<9.017 -11.714 0>, rot=<0.0616 -0.7044 -0.0616 -0.7044>}
 cape {parent=attach_cape, type=Null, loc=<0 5.0323 0>, rot=<0 0 0 1>}
 base_Collar_l {parent=spine_05, loc=<7.9083 4.1201 -6.3571>, rot=<0.9489 0.3006 -0.0045 -0.0963>}
 dyn_Collar_l {parent=base_Collar_l, loc=<0 0 0>, rot=<0 0 0 1>}
 base_Collar_m_l {parent=spine_05, loc=<11.5594 -2.0619 -5.1892>, rot=<0.9119 0.1478 0.0623 0.3779>}
 dyn_Collar_m_l {parent=base_Collar_m_l, loc=<0 0 0>, rot=<0 0 0 1>}
 base_Collar_b {parent=spine_05, loc=<12.4226 -4.406 0>, rot=<-0.7012 -0.0911 -0.0911 -0.7012>}
 dyn_Collar_b {parent=base_Collar_b, loc=<0 0 0>, rot=<0 0 0 1>}
 base_Collar_m_r {parent=spine_05, loc=<11.5698 -1.6989 5.5035>, rot=<-0.203 -0.0497 -0.1843 -0.9604>}
 dyn_Collar_m_r {parent=base_Collar_m_r, loc=<0 0 0>, rot=<0 0 0 1>}
 base_Collar_r {parent=spine_05, loc=<8.3315 3.6289 6.7699>, rot=<0 0 -0.2652 -0.9642>}
 dyn_Collar_r {parent=base_Collar_r, loc=<0 0 0>, rot=<0 0 0 1>}
 head_null {parent=spine_05, type=Null, loc=<20.6573 3.5489 0>, rot=<0 0 -0.036 -0.9994>}
 deform_pec_aim_base_l {parent=spine_04, loc=<3.2456 9.3309 -1.2237>, rot=<0.0603 -0.4033 0.3318 -0.8507>}
 deform_pec_aim_l {parent=deform_pec_aim_base_l, loc=<0 0 0>, rot=<0 0 0 -1>}
 deform_pec_base_l {parent=deform_pec_aim_l, loc=<9.3665 0 0>, rot=<0 0 0 -1>}
 deform_pec_l {parent=deform_pec_base_l, loc=<0 0 0>, rot=<0 0 0 -1>}
 deform_pec_aim_base_r {parent=spine_04, loc=<3.2452 9.3309 1.2237>, rot=<-0.4033 -0.0603 0.8507 0.3318>}
 deform_pec_aim_r {parent=deform_pec_aim_base_r, loc=<0 0 0>, rot=<0 0 0 -1>}
 deform_pec_base_r {parent=deform_pec_aim_r, loc=<-9.3667 -0.0003 -0.0004>, rot=<0 0 0 -1>}
 deform_pec_r {parent=deform_pec_base_r, loc=<0 0 0>, rot=<0 0 0 -1>}
 base_coat_a_l {parent=spine_04, loc=<-3.0192 9.1045 -9.2083>, rot=<0.0397 -0.9388 -0.3423 -0.0019>}
 dyn_coat_a_l_01 {parent=base_coat_a_l, loc=<0 0 0>, rot=<0 0 0 1>}
 dyn_coat_a_l_02 {parent=dyn_coat_a_l_01, loc=<5.3873 0 0>, rot=<0 0 0 1>}
 base_coat_b_l {parent=spine_04, loc=<-2.4274 1.2381 -11.6092>, rot=<0.036 -0.9992 -0.0006 -0.0172>}
 dyn_coat_b_l_01 {parent=base_coat_b_l, loc=<0 0 0>, rot=<0 0 0 1>}
 dyn_coat_b_l_02 {parent=dyn_coat_b_l_01, loc=<5.2164 0 0>, rot=<0 0 0 1>}
 base_coat_c_l {parent=spine_04, loc=<-2.0753 -6.2491 -7.8631>, rot=<0.0246 -0.9059 0.4217 -0.0317>}
 dyn_coat_c_l_01 {parent=base_coat_c_l, loc=<0 0 0>, rot=<0 0 0 1>}
 dyn_coat_c_l_02 {parent=dyn_coat_c_l_01, loc=<5.2478 0 0>, rot=<0 0 0 1>}
 base_coat_d_b {parent=spine_04, loc=<-2.133 -7.9479 0>, rot=<0.0373 -0.7061 0.707 -0.0141>}
 dyn_coat_d_b_01 {parent=base_coat_d_b, loc=<0 0 0>, rot=<0 0 0 1>}
 dyn_coat_d_b_02 {parent=dyn_coat_d_b_01, loc=<5.0223 0 0>, rot=<0 0 0 1>}
 base_coat_a_r {parent=spine_04, loc=<-3.1073 9.0981 9.211>, rot=<0.0019 0.3423 0.9388 -0.0397>}
 dyn_coat_a_r_01 {parent=base_coat_a_r, loc=<0 0 0>, rot=<0 0 0 1>}
 dyn_coat_a_r_02 {parent=dyn_coat_a_r_01, loc=<5.5069 0 0>, rot=<0 0 0 1>}
 base_coat_b_r {parent=spine_04, loc=<-2.0088 1.2683 11.5947>, rot=<0.0172 0.0006 0.9992 -0.036>}
 dyn_coat_b_r_01 {parent=base_coat_b_r, loc=<0 0 0>, rot=<0 0 0 1>}
 dyn_coat_b_r_02 {parent=dyn_coat_b_r_01, loc=<5.8911 0 0>, rot=<0 0 0 1>}
 base_coat_c_r {parent=spine_04, loc=<-2.0751 -6.2491 7.8631>, rot=<0.0337 -0.499 0.8657 -0.0217>}
 dyn_coat_c_r_01 {parent=base_coat_c_r, loc=<0 0 0>, rot=<0 0 0 1>}
 dyn_coat_c_r_02 {parent=dyn_coat_c_r_01, loc=<5.3726 0 0>, rot=<0 0 0 1>}
 thigh_l {parent=pelvis, loc=<-1.0855 -0.2109 -8.7053>, rot=<0.0009 0.0355 0.0155 -0.9993>}
 calf_l {parent=thigh_l, loc=<-42.5284 -0.0019 -0.7297>, rot=<0.0006 -0.0161 -0.0391 -0.9991>}
 foot_l {parent=calf_l, loc=<-46.7401 -2.5014 -0.592>, rot=<0.0427 -0.0238 0.0353 -0.9982>}
 ball_l {parent=foot_l, loc=<-3.5658 10.8066 -0.1544>, rot=<0.0001 -0.0001 -0.7063 -0.7079>}
 deform_foot_out_base_l {parent=foot_l, loc=<1.7153 -4.2488 0.3122>, rot=<0.6469 -0.7618 0.0318 0.0123>}
 deform_foot_out_l {parent=deform_foot_out_base_l, loc=<0 0 0>, rot=<0 0 0 1>}
 calf_twist_02_l {parent=calf_l, loc=<-19.1546 -1.7943 -0.2638>, rot=<0.0237 0.0095 -0.0035 -0.9997>}
 calf_twist_01_l {parent=calf_l, loc=<-35.8483 -1.9197 -0.5794>, rot=<0.0237 0.0095 -0.0035 -0.9997>}
 deform_knee_in_base_l {parent=calf_l, loc=<0.8031 -5.0231 0.2456>, rot=<0 0 0 -1>}
 deform_knee_in_l {parent=deform_knee_in_base_l, loc=<0 0 0>, rot=<0 0 0 -1>}
 deform_knee_out_base_l {parent=calf_l, loc=<0 4 0>, rot=<0 0 0 -1>}
 deform_knee_out_l {parent=deform_knee_out_base_l, loc=<0 0 0>, rot=<0 0 0 1>}
 thigh_twist_01_l {parent=thigh_l, loc=<-27.952 0 0>, rot=<0.0013 0 0 -1>}
 deform_groin_base_l {parent=thigh_l, loc=<0.4801 5.6776 -0.2486>, rot=<0 0 0 -1>}
 deform_groin_l {parent=deform_groin_base_l, loc=<0 0 0>, rot=<0 0 0 -1>}
 thigh_r {parent=pelvis, loc=<-1.0855 -0.2109 8.7053>, rot=<0.0355 -0.0009 0.9993 0.0155>}
 calf_r {parent=thigh_r, loc=<42.5284 0.0019 0.7297>, rot=<0.0006 -0.0161 -0.0391 -0.9991>}
 foot_r {parent=calf_r, loc=<46.7401 2.5014 0.592>, rot=<0.0427 -0.0238 0.0353 -0.9982>}
 ball_r {parent=foot_r, loc=<3.5658 -10.8066 0.1544>, rot=<0.0001 -0.0001 -0.7063 -0.7079>}
 deform_foot_out_base_r {parent=foot_r, loc=<-1.7153 4.2488 -0.3122>, rot=<-0.559 -0.8285 0.0204 -0.0273>}
 deform_foot_out_r {parent=deform_foot_out_base_r, loc=<0 0 0>, rot=<0 0 0 1>}
 calf_twist_02_r {parent=calf_r, loc=<19.1546 1.7943 0.2638>, rot=<0.0237 0.0095 -0.0035 -0.9997>}
 calf_twist_01_r {parent=calf_r, loc=<35.8483 1.9197 0.5794>, rot=<0.0237 0.0095 -0.0035 -0.9997>}
 deform_knee_in_base_r {parent=calf_r, loc=<-0.8031 5.0231 -0.2457>, rot=<0 0 0 -1>}
 deform_knee_in_r {parent=deform_knee_in_base_r, loc=<0 0 0>, rot=<0 0 0 -1>}
 deform_knee_out_base_r {parent=calf_r, loc=<0 -4 0>, rot=<0 0 0 -1>}
 deform_knee_out_r {parent=deform_knee_out_base_r, loc=<0 0 0>, rot=<0 0 0 1>}
 thigh_twist_01_r {parent=thigh_r, loc=<27.952 0 0>, rot=<0.0013 0 0 -1>}
 deform_groin_base_r {parent=thigh_r, loc=<-0.4801 -5.6776 0.2486>, rot=<0 0 0 -1>}
 deform_groin_r {parent=deform_groin_base_r, loc=<0 0 0>, rot=<0 0 0 -1>}
 deform_glute_base_l {parent=pelvis, loc=<-7.3466 -9.0213 -9.1245>, rot=<0.0172 0.031 -0.4477 -0.8935>}
 deform_glute_l {parent=deform_glute_base_l, loc=<0 0 0>, rot=<0 0 0 1>}
 deform_glute_base_r {parent=pelvis, loc=<-7.3464 -9.0213 9.1245>, rot=<0.031 -0.0172 0.8935 -0.4477>}
 deform_glute_r {parent=deform_glute_base_r, loc=<0 0 0>, rot=<0 0 0 1>}
 base_belt {parent=pelvis, loc=<7.7711 0.2519 -10.267>, rot=<-0.0175 -0.9778 0.0037 -0.2088>}
 dyn_belt {parent=base_belt, loc=<0 0 0>, rot=<0 0 0 1>}
 dyn_belt_01 {parent=dyn_belt, loc=<4.1422 0 0>, rot=<0 0 0 1>}
 dyn_belt_02 {parent=dyn_belt_01, loc=<3.7006 0 -0.0677>, rot=<0 0 0 1>}
 base_belt_b_01 {parent=dyn_belt_01, loc=<0 0 0>, rot=<0 0 0 1>}
 dyn_belt_b_01 {parent=base_belt_b_01, loc=<0 -7.2161 -1.1173>, rot=<-0.0419 -0.1358 -0.0058 -0.9898>}
 base_belt_f_01 {parent=dyn_belt_01, loc=<0 0 0>, rot=<0 0 0 1>}
 dyn_belt_f_01 {parent=base_belt_f_01, loc=<0 5.6384 -2.8328>, rot=<0.2899 -0.1262 -0.0506 -0.9473>}
]
variables [
 Crouch walk {type=double, default=0}
 HelmetOn {type=bool, default=False}
 OnlyCrouch {type=bool, default=False}
 Pistol {type=double, default=0}
 Riffle {type=double, default=0}
 Targeting {type=double, default=0}
 parachute open {type=bool, default=False}
]
