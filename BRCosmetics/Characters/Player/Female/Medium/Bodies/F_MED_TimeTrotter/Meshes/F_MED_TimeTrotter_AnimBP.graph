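# Animation Blueprint for the TimeTrotter female medium cosmetic body, layering procedural correction and coat motion onto the parent mesh's pose.
# The copied pose feeds a Control Rig, then seven chained spring-driven AnimDynamics nodes simulate coat bones.
# Update logic derives crouch, targeting and helmet inputs, the helmet from the Stage2 Parts variant, and sets gravity override variables.
# Weapon changes map the core animation type to rifle, pistol or melee values.
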AnimBlueprint /BRCosmetics/Characters/Player/Female/Medium/Bodies/F_MED_TimeTrotter/Meshes/F_MED_TimeTrotter_AnimBP
target_skeleton=&/F_MED_TimeTrotter_Skeleton
graphs [
 AnimGraph {nodes [
   n1 = Control Rig!(
    Source=Copy Pose From Mesh() {bUseAttachedParent=true}
    Alpha
    Riffle:double=Get Rifle() -> (Rifle:double)
    Pistol:double=Get pistol() -> (pistol:double)
    Targeting:double=Get Targeting() -> (Targeting:double)
    Crouch walk:double=Get CrouchWalk() -> (CrouchWalk:double)
    HelmetOn:bool=Get Helmet() -> (Helmet:bool)
    parachute open:bool=Get bIsParachuteOpened()
    OnlyCrouch:bool=Get OnlyCrouch() -> (OnlyCrouch:bool)
   ) -> (Pose) {controlRigAssetReference {blueprintRigClass=&/F_MED_TimeTrotter_CtrlRig.F_MED_TimeTrotter_CtrlRig_C}}
   n2 = Anim Dynamics - Bone: dyn_coat_a_l_01!(ComponentPose=Local To Component(LocalPose=@n1), Alpha=1, GravityOverride:Vector="0,0,-980", ExternalForce="0,0,0") -> (Pose) {
    angularDampingOverride=1
    angularSpringConstant=12000
    bAngularSpring=true
    bOverrideAngularDamping=true
    bUseGravityOverride=true
    gravityScale=0.35
    simulationSpace=RootRelative
    boundBone {boneName=dyn_coat_a_l_01}
    componentAppliedLinearAccClamp=<3000 3000 3000>
    componentLinearAccScale=<0.5 0.5 0.5>
    componentLinearVelScale=<0.5 0.5 0.5>
    gravityOverride {z=-980}
    physicsBodyDefinitions [{
      boundBone {boneName=dyn_coat_a_l_01}
      boxExtents=<20 30 30>
      constraintSetup {angularLimitsMax=<0 5 0>, angularLimitsMin=<0 -30 0>, angularTarget=<1 0 0>}
      localJointOffset=<4 0 0>
     }]
    simSpaceSettings {simSpaceAngularAlpha=0.3}
   }
   n3 = Anim Dynamics - Bone: dyn_coat_b_l_01!(ComponentPose=@n2, Alpha=1, GravityOverride:Vector="0,0,-980", ExternalForce="0,0,0") -> (Pose) {
    angularDampingOverride=1
    angularSpringConstant=12000
    bAngularSpring=true
    bOverrideAngularDamping=true
    bUseGravityOverride=true
    gravityScale=0.35
    simulationSpace=RootRelative
    boundBone {boneName=dyn_coat_b_l_01}
    componentAppliedLinearAccClamp=<3000 3000 3000>
    componentLinearAccScale=<0.5 0.5 0.5>
    componentLinearVelScale=<0.5 0.5 0.5>
    gravityOverride {z=-980}
    physicsBodyDefinitions [{
      boundBone {boneName=dyn_coat_b_l_01}
      boxExtents=<20 30 30>
      constraintSetup {angularLimitsMax=<0 5 0>, angularLimitsMin=<0 -30 0>, angularTarget=<1 0 0>}
      localJointOffset=<4 0 0>
     }]
    simSpaceSettings {simSpaceAngularAlpha=0.3}
   }
   n4 = Anim Dynamics - Bone: dyn_coat_c_l_01!(ComponentPose=@n3, Alpha=1, GravityOverride:Vector="0,0,-980", ExternalForce="0,0,0") -> (Pose) {
    angularDampingOverride=1
    angularSpringConstant=12000
    bAngularSpring=true
    bOverrideAngularDamping=true
    bUseGravityOverride=true
    gravityScale=0.35
    simulationSpace=RootRelative
    boundBone {boneName=dyn_coat_c_l_01}
    componentAppliedLinearAccClamp=<3000 3000 3000>
    componentLinearAccScale=<0.5 0.5 0.5>
    componentLinearVelScale=<0.5 0.5 0.5>
    gravityOverride {z=-980}
    physicsBodyDefinitions [{
      boundBone {boneName=dyn_coat_c_l_01}
      boxExtents=<20 30 30>
      constraintSetup {angularLimitsMax=<0 5 0>, angularLimitsMin=<0 -30 0>, angularTarget=<1 0 0>}
      localJointOffset=<4 0 0>
     }]
    simSpaceSettings {simSpaceAngularAlpha=0.3}
   }
   n5 = Anim Dynamics - Bone: dyn_coat_d_b_01!(ComponentPose=@n4, Alpha=1, GravityOverride:Vector="0,0,-980", ExternalForce="0,0,0") -> (Pose) {
    angularDampingOverride=1
    angularSpringConstant=12000
    bAngularSpring=true
    bOverrideAngularDamping=true
    bUseGravityOverride=true
    gravityScale=0.35
    simulationSpace=RootRelative
    boundBone {boneName=dyn_coat_d_b_01}
    componentAppliedLinearAccClamp=<3000 3000 3000>
    componentLinearAccScale=<0.5 0.5 0.5>
    componentLinearVelScale=<0.5 0.5 0.5>
    gravityOverride {z=-980}
    physicsBodyDefinitions [{
      boundBone {boneName=dyn_coat_d_b_01}
      boxExtents=<20 30 30>
      constraintSetup {angularLimitsMax=<0 5 0>, angularLimitsMin=<0 -30 0>, angularTarget=<1 0 0>}
      localJointOffset=<4 0 0>
     }]
    simSpaceSettings {simSpaceAngularAlpha=0.3}
   }
   n6 = Anim Dynamics - Bone: dyn_coat_c_r_01!(ComponentPose=@n5, Alpha=1, GravityOverride:Vector="0,0,-980", ExternalForce="0,0,0") -> (Pose) {
    angularDampingOverride=1
    angularSpringConstant=12000
    bAngularSpring=true
    bOverrideAngularDamping=true
    bUseGravityOverride=true
    gravityScale=0.35
    simulationSpace=RootRelative
    boundBone {boneName=dyn_coat_c_r_01}
    componentAppliedLinearAccClamp=<3000 3000 3000>
    componentLinearAccScale=<0.5 0.5 0.5>
    componentLinearVelScale=<0.5 0.5 0.5>
    gravityOverride {z=-980}
    physicsBodyDefinitions [{
      boundBone {boneName=dyn_coat_c_r_01}
      boxExtents=<20 30 30>
      constraintSetup {angularLimitsMax=<0 5 0>, angularLimitsMin=<0 -30 0>, angularTarget=<1 0 0>}
      localJointOffset=<4 0 0>
     }]
    simSpaceSettings {simSpaceAngularAlpha=0.3}
   }
   n7 = Anim Dynamics - Bone: dyn_coat_b_r_01!(ComponentPose=@n6, Alpha=1, GravityOverride:Vector="0,0,-980", ExternalForce="0,0,0") -> (Pose) {
    angularDampingOverride=1
    angularSpringConstant=12000
    bAngularSpring=true
    bOverrideAngularDamping=true
    bUseGravityOverride=true
    gravityScale=0.35
    simulationSpace=RootRelative
    boundBone {boneName=dyn_coat_b_r_01}
    componentAppliedLinearAccClamp=<3000 3000 3000>
    componentLinearAccScale=<0.5 0.5 0.5>
    componentLinearVelScale=<0.5 0.5 0.5>
    gravityOverride {z=-980}
    physicsBodyDefinitions [{
      boundBone {boneName=dyn_coat_b_r_01}
      boxExtents=<20 30 30>
      constraintSetup {angularLimitsMax=<0 5 0>, angularLimitsMin=<0 -30 0>, angularTarget=<1 0 0>}
      localJointOffset=<4 0 0>
     }]
    simSpaceSettings {simSpaceAngularAlpha=0.3}
   }
   n8 = Anim Dynamics - Bone: dyn_coat_a_r_01!(ComponentPose=@n7, Alpha=1, GravityOverride:Vector="0,0,-980", ExternalForce="0,0,0") -> (Pose) {
    angularDampingOverride=1
    angularSpringConstant=12000
    bAngularSpring=true
    bOverrideAngularDamping=true
    bUseGravityOverride=true
    gravityScale=0.35
    simulationSpace=RootRelative
    boundBone {boneName=dyn_coat_a_r_01}
    componentAppliedLinearAccClamp=<3000 3000 3000>
    componentLinearAccScale=<0.5 0.5 0.5>
    componentLinearVelScale=<0.5 0.5 0.5>
    gravityOverride {z=-980}
    physicsBodyDefinitions [{
      boundBone {boneName=dyn_coat_a_r_01}
      boxExtents=<20 30 30>
      constraintSetup {angularLimitsMax=<0 5 0>, angularLimitsMin=<0 -30 0>, angularTarget=<1 0 0>}
      localJointOffset=<4 0 0>
     }]
    simSpaceSettings {simSpaceAngularAlpha=0.3}
   }
   Output Pose(Result=Component To Local(ComponentPose=@n8)) {layerGroup=None}
  ]}
 EventGraph {nodes [
   Event BlueprintUpdateAnimation()
   > n1 = Sequence!(execute) -> (then_0, then_1, then_2:Exec, then_3:Exec, then_4:Exec)
   > n2 = "Gravity Override State Function (Parameter Asset) (Nativized)"(CharacterPartAnimInstance=Self-Reference(), GameTimeInSeconds=GetGameTimeInSeconds())
   > Set GravityOverride(GravityOverride=@n2)
   > Set RBAN Settings Out()
   n3 = Get bIsCrouching()
   n4 = Get bIsMoving2D()
   Set CrouchWalk(execute=@n1.then_1, CrouchWalk:double="To Float (Boolean)"(InBool=AND Boolean(A=@n3, B=@n4))) -> (Output_Get:double)
   Set Targeting(execute=@n1.then_2, Targeting:double="To Float (Boolean)"(InBool=Get bIsTargeting())) -> (Output_Get:double)
   n5 = Cast To FortPlayerPawn(execute=@n1.then_3, Object=GetOwningActor())
   > n6 = GetVariantChannelActiveIndexForItemActor!(
    execute
    Item:Athena Cosmetic Item Definition Object Reference=/BRCosmetics/Athena/Items/Cosmetics/Characters/Character_TimeTrotter
    Actor:Actor Object Reference=@n5
    Channel:Gameplay Tag Structure=(TagName="Cosmetics.Variant.Channel.Parts")
    VariantList:Array of Gameplay Tag Structures=Make Array("[0]"=(TagName="Cosmetics.Variant.Property.Stage2"))
    bErrorIfChannelMissing:bool=false
    bErrorIfTagNotInArray:bool=false
   ) -> (then, Index:int=0, bIndexIsValid:bool, bFoundChannel:bool)
   > n7 = Branch(Condition=@n6.bIndexIsValid)
   > Set Helmet(Helmet:bool=true) -> (Output_Get:bool)
   Set Helmet(execute=@n7.else, Helmet=false)
   n8 = OR Boolean!(A=Get bIsSliding(), B=Get bIsBaseballSliding(), C:bool=Get bIsMoving2D()) -> (ReturnValue)
   Set OnlyCrouch(execute=@n1.then_4, OnlyCrouch:bool=AND Boolean(A=Get bIsCrouching(), B=NOT Boolean(A=@n8))) -> (Output_Get:bool)
   Event OnWeaponChanged()
   > n9 = Cast To FortPlayerAnimInstance(Object=GetAnimInstance(self=GetOwnerSkeletalMeshComponent()))
   > Set pistol(pistol:double=0) -> (Output_Get:double)
   > Set Melee(Melee:double=0) -> (Output_Get:double)
   > Set Rifle(Rifle:double=0) -> (Output_Get:double)
   > n10 = Switch on EFortWeaponCoreAnimation!(execute, Selection:EFortWeaponCoreAnimation Enum=Get WeaponCoreAnim(self=@n9)) -> (
    Melee:Exec
    Pistol:Exec
    Shotgun:Exec
    PaperBlueprint:Exec
    Rifle:Exec
    MeleeOneHand:Exec
    MachinePistol:Exec
    RocketLauncher:Exec
    GrenadeLauncher:Exec
    GoingCommando:Exec
    AssaultRifle:Exec
    TacticalShotgun:Exec
    SniperRifle:Exec
    TrapPlacement:Exec
    ShoulderLauncher:Exec
    AbilityDecoTool:Exec
    Crossbow:Exec
    C4:Exec
    RemoteControl:Exec
    DualWield:Exec
    AR_BullPup:Exec
    AR_ForwardGrip:Exec
    MedPackPaddles:Exec
    SMG_P90:Exec
    AR_DrumGun:Exec
    Consumable_Small:Exec
    Consumable_Large:Exec
    Balloon:Exec
    MountedTurret:Exec
    CreativeTool:Exec
    ExplosiveBow:Exec
    AshtonIndigo:Exec
    AshtonChicago:Exec
    MeleeDualWield:Exec
    Unarmed:Exec
   )
   Set Rifle(
    execute=[
     @n10.AR_BullPup
     @n10.AR_DrumGun
     @n10.AR_ForwardGrip
     @n10.AssaultRifle
     @n10.Crossbow
     @n10.GrenadeLauncher
     @n10.Rifle
     @n10.RocketLauncher
     @n10.SMG_P90
     @n10.Shotgun
     @n10.SniperRifle
     @n10.TacticalShotgun
    ]
    Rifle=1
   )
   > Set Only_Rifle(Only_Rifle:bool=1) -> (Output_Get:bool)
   > Set Only_Pistol(Only_Pistol:bool=false) -> (Output_Get:bool)
   Set pistol(execute=[@n10.MachinePistol, @n10.Pistol], pistol=1)
   > Set Only_Rifle(Only_Rifle=false)
   > Set Only_Pistol(Only_Pistol=true)
   Set Melee(execute=@n10.Melee, Melee=1)
   > Set Only_Rifle(Only_Rifle=false)
   > Set Only_Pistol(Only_Pistol=false)
   Reroute Node(InputPin:Wildcard) -> (OutputPin:Wildcard)
  ]}
]
variables [
 CrouchWalk {type=number}
 GravityOverride {type=Vector}
 Helmet {type=boolean}
 Melee {type=number}
 Only_Pistol {type=boolean}
 Only_Rifle {type=boolean}
 OnlyCrouch {type=boolean}
 pistol {type=number}
 RBAN Settings Out {type=FortRigidBodyAnimNodeParameters}
 Rifle {type=number}
 Targeting {type=number}
]
