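# Control Rig for the female TimeTrotter face accessory that applies corrective offsets to ponytail and earring bones.
# A spherical pose reader on head relative to spine_05, multiplied by Crouch walk and Targeting, weights ponytail offsets, with a second set when Cap crouch is true.
# A DeltaRotation function remaps head swing against head_null into weights for two opposite-direction earring offsets.
# The Front end flag switches on a further fixed earring offset.
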ControlRigBlueprint /BRCosmetics/Characters/Player/Female/Medium/Bodies/F_MED_TimeTrotter/Meshes/Parts/F_MED_TimeTrotter_FaceAcc_CtrlRig
graphs [
 Multiply_1_ContainedGraph {nodes [
   n1 = RigVMFunctionEntryNode()
   Multiply = &RigVMFunction_MathDoubleMul(A=@n1.A, B=@n1.B)
   RigVMFunctionReturnNode(Result=&RigVMFunction_MathDoubleMul(B=@n1.C, A=@Multiply.Result).Result)
  ]}
 Multiply_1_ContainedGraph {nodes [
   n1 = RigVMFunctionEntryNode()
   Multiply = &RigVMFunction_MathDoubleMul(A=@n1.A, B=@n1.B)
   RigVMFunctionReturnNode(Result=&RigVMFunction_MathDoubleMul(B=@n1.C, A=@Multiply.Result).Result)
  ]}
 New Function_ContainedGraph {nodes [
   n1 = RigVMFunctionEntryNode()
   RigVMFunction_MathBoolToInteger = &RigVMFunction_MathBoolToInteger(Value=@n1.onlyTwist)
   RigVMFunction_MathQuaternionSwingTwist = &RigVMFunction_MathQuaternionSwingTwist(
    twistAxis=@n1.twistAxis
    Input=&RigUnit_GetRelativeTransformForItem(bChildInitial=False, bParentInitial=False, child=@n1.child, parent=@n1.parent).RelativeTransform.Rotation
   )
   n2 = RigVMDispatchNode(
    Values=((),())
    Index=@RigVMFunction_MathBoolToInteger.Result
    Values.0=@RigVMFunction_MathQuaternionSwingTwist.Swing
    Values.1=@RigVMFunction_MathQuaternionSwingTwist.Twist
   ) {template="DISPATCH_RigVMDispatch_SelectInt32(in Index,in Values,out Result)"}
   RigVMFunction_MathQuaternionToEuler = &RigVMFunction_MathQuaternionToEuler(Value=@n2.Result, RotationOrder=@n1.Rotation Order)
   > DISPATCH_RigVMDispatch_Print = RigVMDispatchNode(ScreenColor=(R=1,G=1,B=1,A=1), ScreenDuration=1.5, Enabled=@n1.debugRotation, Value=@RigVMFunction_MathQuaternionToEuler.Result) {template="DISPATCH_RigVMDispatch_Print(in Prefix,in Value,in Enabled,in ScreenDuration,in ScreenColor)"}
   Cast = RigVMDispatchNode(Value=@n1.triggerAxis) {template="DISPATCH_RigVMDispatch_CastEnumToInt(in Value,out Result)"}
   DISPATCH_RigVMDispatch_SelectInt32 = RigVMDispatchNode(
    Values=(0,0,0)
    Index=@Cast.Result
    Values.0=@RigVMFunction_MathQuaternionToEuler.Result.X
    Values.1=@RigVMFunction_MathQuaternionToEuler.Result.Y
    Values.2=@RigVMFunction_MathQuaternionToEuler.Result.Z
   ) {template="DISPATCH_RigVMDispatch_SelectInt32(in Index,in Values,out Result)"}
   Remap = &RigVMFunction_MathDoubleRemap(
    bClamp=False
    Value=@DISPATCH_RigVMDispatch_SelectInt32.Result
    TargetMinimum=@n1.outputMinMax.X
    TargetMaximum=@n1.outputMinMax.Y
    SourceMinimum=@n1.rotationMinMax.X
    SourceMaximum=@n1.rotationMinMax.Y
   )
   Absolute = &RigVMFunction_MathDoubleAbs(Value=@Remap.Result)
   Multiply = &RigVMFunction_MathDoubleMul(
    A=&RigVMFunction_MathDoubleClamp(Value=RigVMDispatchNode(True=@Absolute.Result, Condition=@n1.absoluteValue, False=@Remap.Result) {template="DISPATCH_RigVMDispatch_If(in Condition,in True,in False,out Result)"}.Result).Result
    B=@n1.premult
   )
   > RigVMFunctionReturnNode(outputQuat=@n2.Result, outputWeight=@Multiply.Result)
  ]}
 New Function_ContainedGraph {nodes [
   n1 = RigVMFunctionEntryNode()
   RigVMFunction_MathBoolToInteger = &RigVMFunction_MathBoolToInteger(Value=@n1.onlyTwist)
   RigVMFunction_MathQuaternionSwingTwist = &RigVMFunction_MathQuaternionSwingTwist(
    twistAxis=@n1.twistAxis
    Input=&RigUnit_GetRelativeTransformForItem(bChildInitial=False, bParentInitial=False, child=@n1.child, parent=@n1.parent).RelativeTransform.Rotation
   )
   n2 = RigVMDispatchNode(
    Values=((),())
    Index=@RigVMFunction_MathBoolToInteger.Result
    Values.0=@RigVMFunction_MathQuaternionSwingTwist.Swing
    Values.1=@RigVMFunction_MathQuaternionSwingTwist.Twist
   ) {template="DISPATCH_RigVMDispatch_SelectInt32(in Index,in Values,out Result)"}
   RigVMFunction_MathQuaternionToEuler = &RigVMFunction_MathQuaternionToEuler(Value=@n2.Result, RotationOrder=@n1.Rotation Order)
   > DISPATCH_RigVMDispatch_Print = RigVMDispatchNode(ScreenColor=(R=1,G=1,B=1,A=1), ScreenDuration=1.5, Enabled=@n1.debugRotation, Value=@RigVMFunction_MathQuaternionToEuler.Result) {template="DISPATCH_RigVMDispatch_Print(in Prefix,in Value,in Enabled,in ScreenDuration,in ScreenColor)"}
   Cast = RigVMDispatchNode(Value=@n1.triggerAxis) {template="DISPATCH_RigVMDispatch_CastEnumToInt(in Value,out Result)"}
   DISPATCH_RigVMDispatch_SelectInt32 = RigVMDispatchNode(
    Values=(0,0,0)
    Index=@Cast.Result
    Values.0=@RigVMFunction_MathQuaternionToEuler.Result.X
    Values.1=@RigVMFunction_MathQuaternionToEuler.Result.Y
    Values.2=@RigVMFunction_MathQuaternionToEuler.Result.Z
   ) {template="DISPATCH_RigVMDispatch_SelectInt32(in Index,in Values,out Result)"}
   Remap = &RigVMFunction_MathDoubleRemap(
    bClamp=False
    Value=@DISPATCH_RigVMDispatch_SelectInt32.Result
    TargetMinimum=@n1.outputMinMax.X
    TargetMaximum=@n1.outputMinMax.Y
    SourceMinimum=@n1.rotationMinMax.X
    SourceMaximum=@n1.rotationMinMax.Y
   )
   Absolute = &RigVMFunction_MathDoubleAbs(Value=@Remap.Result)
   Multiply = &RigVMFunction_MathDoubleMul(
    A=&RigVMFunction_MathDoubleClamp(Value=RigVMDispatchNode(True=@Absolute.Result, Condition=@n1.absoluteValue, False=@Remap.Result) {template="DISPATCH_RigVMDispatch_If(in Condition,in True,in False,out Result)"}.Result).Result
    B=@n1.premult
   )
   > RigVMFunctionReturnNode(outputQuat=@n2.Result, outputWeight=@Multiply.Result)
  ]}
 RigVMFunctionLibrary {nodes [
   DeltaRotation v4 = RigVMCollapseNode(
    ExecuteContext=()
    Rotation Order=ZYX
    child=(Type=Bone,Name="None")
    outputMinMax=(X=0,Y=1)
    parent=(Type=Bone,Name="None")
    premult=1
    rotationMinMax=(X=0,Y=0)
    twistAxis=(X=1,Y=0,Z=0)
   )
   DeltaRotation v4_1 = RigVMCollapseNode(
    ExecuteContext=()
    Rotation Order=ZYX
    child=(Type=Bone,Name="None")
    outputMinMax=(X=0,Y=1)
    parent=(Type=Bone,Name="None")
    premult=1
    rotationMinMax=(X=0,Y=0)
    twistAxis=(X=1,Y=0,Z=0)
   )
  ]}
 RigVMFunction_Sequence_1_ContainedGraph {nodes [
   n1 = RigVMFunctionEntryNode()
   RigVMFunction_Sequence = &RigVMFunction_Sequence(ExecuteContext=@n1.ExecuteContext)
   n2 = &RigVMFunction_Sequence(ExecuteContext=@RigVMFunction_Sequence.B)
   RigVMFunctionReturnNode(A=@RigVMFunction_Sequence.A, B=@n2.A, C=@n2.B)
  ]}
 RigVMModel {nodes [
   RigUnit_BeginExecution = &RigUnit_BeginExecution()
   RigVMFunction_Sequence = RigVMAggregateNode(ExecuteContext=@RigUnit_BeginExecution.ExecutePin)
   n1 = &RigUnit_SphericalPoseReader(
    ActiveRegionScaleFactors=(PositiveWidth=1,NegativeWidth=1,PositiveHeight=1,NegativeHeight=1)
    ActiveRegionSize=0.1
    Debug=(bDrawDebug=false,bDraw2D=False,bDrawLocalAxes=False,DebugScale=25,DebugSegments=20,DebugThickness=0.25)
    DriverAxis=(X=1,Y=0,Z=0)
    DriverItem=(Type=Bone,Name="head")
    FalloffRegionScaleFactors=(PositiveWidth=1,NegativeWidth=1,PositiveHeight=1,NegativeHeight=1)
    FalloffSize=0.4
    FlipHeightScaling=False
    FlipWidthScaling=False
    OptionalParentItem=(Type=Bone,Name="spine_05")
    RotationOffset=(X=-55,Y=40,Z=50)
    ExecutePin=@RigVMFunction_Sequence.A
   )
   Multiply = RigVMAggregateNode(A=@n1.OutputParam, B=RigVMVariableNode() {variable=Crouch walk}.Value, C=RigVMVariableNode() {variable=Targeting}.Value)
   > &RigUnit_ModifyTransforms(
    ItemToModify="((Item=(Type=Bone,Name=\"dyn_ponytail_02\"),Transform=(Rotation=(X=0,Y=0,Z=0,W=1),Translation=(X=0,Y=2,Z=0),Scale3D=(X=1,Y=1,Z=1))),(Item=(Type=Bone,Name=\"dyn_ponytail_03\"),Transform=(Rotation=(X=0,Y=0,Z=0,W=1),Translation=(X=0,Y=2,Z=0),Scale3D=(X=1,Y=1,Z=1) ...(truncated)"
    WeightMinimum=0
    Weight=@Multiply.Result
   )
   RigVMFunction_ControlFlowBranch = &RigVMFunction_ControlFlowBranch(ExecuteContext=@RigVMFunction_Sequence.B, Condition=RigVMVariableNode() {variable=Cap crouch}.Value)
   n2 = &RigUnit_SphericalPoseReader(
    ActiveRegionScaleFactors=(PositiveWidth=1,NegativeWidth=1,PositiveHeight=1,NegativeHeight=1)
    ActiveRegionSize=0.1
    Debug=(bDrawDebug=false,bDraw2D=False,bDrawLocalAxes=False,DebugScale=25,DebugSegments=20,DebugThickness=0.25)
    DriverAxis=(X=1,Y=0,Z=0)
    DriverItem=(Type=Bone,Name="head")
    FalloffRegionScaleFactors=(PositiveWidth=1,NegativeWidth=1,PositiveHeight=1,NegativeHeight=1)
    FalloffSize=0.4
    FlipHeightScaling=False
    FlipWidthScaling=False
    OptionalParentItem=(Type=Bone,Name="spine_05")
    RotationOffset=(X=-55,Y=40,Z=50)
    ExecutePin=@RigVMFunction_ControlFlowBranch.True
   )
   Multiply_1 = RigVMAggregateNode(A=@n2.OutputParam, B=RigVMVariableNode() {variable=Crouch walk}.Value, C=RigVMVariableNode() {variable=Targeting}.Value)
   > &RigUnit_ModifyTransforms(
    ItemToModify="((Item=(Type=Bone,Name=\"dyn_ponytail_02\"),Transform=(Rotation=(X=0,Y=-0.0872,Z=0,W=0.9962),Translation=(X=0,Y=3,Z=0),Scale3D=(X=1,Y=1,Z=1))),(Item=(Type=Bone,Name=\"dyn_ponytail_03\"),Transform=(Rotation=(X=0,Y=0,Z=0,W=1),Translation=(X=0,Y=3,Z=0),Scale3D=(X ...(truncated)"
    WeightMinimum=0
    Weight=@Multiply_1.Result
   )
   DeltaRotation v4_1 = RigVMFunctionReferenceNode(
    Rotation Order=ZYX
    child=(Type=Bone,Name="head")
    outputMinMax=(X=0,Y=1)
    parent=(Type=Null,Name="head_null")
    premult=1
    rotationMinMax=(X=-10,Y=-50)
    triggerAxis=Z
    twistAxis=(X=1,Y=0,Z=0)
    ExecuteContext=@RigVMFunction_Sequence.C
   )
   > &RigUnit_ModifyTransforms(
    ItemToModify="((Item=(Type=Bone,Name=\"base_earring_l\"),Transform=(Rotation=(X=-0.1524,Y=-0.3053,Z=0.3811,W=-0.8593),Translation=(X=0,Y=0,Z=0),Scale3D=(X=1,Y=1,Z=1))),(Item=(Type=Bone,Name=\"base_earring_r\"),Transform=(Rotation=(X=-0.3768,Y=0.0537,Z=0.3768,W=0.8445),Trans ...(truncated)"
    WeightMinimum=0
    Weight=@"DeltaRotation v4_1".outputWeight
   )
   > DeltaRotation v4_2 = RigVMFunctionReferenceNode(
    Rotation Order=ZYX
    child=(Type=Bone,Name="head")
    outputMinMax=(X=0,Y=1)
    parent=(Type=Null,Name="head_null")
    premult=1
    rotationMinMax=(X=0,Y=50)
    triggerAxis=Z
    twistAxis=(X=1,Y=0,Z=0)
   )
   > &RigUnit_ModifyTransforms(
    ItemToModify="((Item=(Type=Bone,Name=\"base_earring_l\"),Transform=(Rotation=(X=0.5597,Y=-0.5217,Z=0.5811,W=0.2771),Translation=(X=0,Y=0,Z=0),Scale3D=(X=1,Y=1,Z=1))),(Item=(Type=Bone,Name=\"base_earring_r\"),Transform=(Rotation=(X=-0.675,Y=-0.1209,Z=-0.6858,W=0.2437),Transl ...(truncated)"
    WeightMinimum=0
    Weight=@"DeltaRotation v4_2".outputWeight
   )
   RigVMFunction_MathBoolToFloat = &RigVMFunction_MathBoolToFloat(Value=RigVMVariableNode() {variable=Front end}.Value)
   > &RigUnit_ModifyTransforms(
    ItemToModify="((Item=(Type=Bone,Name=\"base_earring_l\"),Transform=(Rotation=(X=0.173,Y=-0.0151,Z=-0.0858,W=0.9811),Translation=(X=0,Y=0,Z=0),Scale3D=(X=1,Y=1,Z=1))),(Item=(Type=Bone,Name=\"base_earring_r\"),Transform=(Rotation=(X=-0.173,Y=-0.0151,Z=0.0858,W=0.9811),Transla ...(truncated)"
    WeightMinimum=0
    Weight=@RigVMFunction_MathBoolToFloat.Result
   )
  ]}
]
hierarchy [
 root {loc=<0 0 0>, rot=<0 0 0 1>}
 pelvis {parent=root, loc=<0 -0.5301 98.0571>, rot=<0.0127 0.707 -0.0127 -0.707>}
 spine_01 {parent=pelvis, loc=<6.1208 -0.9328 0>, rot=<0 0 0.034 -0.9994>}
 spine_02 {parent=spine_01, loc=<6.5997 -0.2095 0>, rot=<0 0 0.0238 -0.9997>}
 spine_03 {parent=spine_02, loc=<6.4914 -0.1237 0>, rot=<0 0 0.0149 -0.9999>}
 spine_04 {parent=spine_03, loc=<6.9598 0.0794 0>, rot=<0 0 -0.0132 -0.9999>}
 spine_05 {parent=spine_04, loc=<6.9987 0 0>, rot=<0 0 -0.0015 -1>}
 neck_01 {parent=spine_05, loc=<11.3138 0.1529 0>, rot=<0 0 -0.2003 -0.9797>}
 neck_02 {parent=neck_01, loc=<4.5627 0 0>, rot=<0 0 0.0504 -0.9987>}
 head {parent=neck_02, loc=<5.3914 0 0>, rot=<0 0 0.115 -0.9934>}
 base_ponytail {parent=head, loc=<11.6238 -7.3448 0>, rot=<0.5974 -0.3783 0.3783 -0.5974>}
 dyn_ponytail_01 {parent=base_ponytail, loc=<0 0 0>, rot=<0 0 0 1>}
 dyn_ponytail_02 {parent=dyn_ponytail_01, loc=<9.4763 0 0>, rot=<0.0074 -0.8426 0.0047 -0.5384>}
 dyn_ponytail_03 {parent=dyn_ponytail_02, loc=<8 0 0>, rot=<0 0 0 1>}
 dyn_ponytail_04 {parent=dyn_ponytail_03, loc=<8 0 0>, rot=<0 0 0 1>}
 dyn_ponytail_05 {parent=dyn_ponytail_04, loc=<8 0 0>, rot=<0 0 0 1>}
 dyn_ponytail_06 {parent=dyn_ponytail_05, loc=<8 0 0>, rot=<0 0 0 1>}
 dyn_ponytail_07 {parent=dyn_ponytail_06, loc=<8 0 0>, rot=<0 0 0 1>}
 base_earring_l {parent=head, loc=<-0.897 0.1238 -6.5497>, rot=<0.0011 -0.9596 -0.2815 0.0017>}
 dyn_earring_l {parent=base_earring_l, loc=<0 0 0>, rot=<0 0 0 1>}
 base_earring_r {parent=head, loc=<-0.8973 0.1238 6.5497>, rot=<-0.0018 0.2989 0.9543 -0.0011>}
 dyn_earring_r {parent=base_earring_r, loc=<0 0 0>, rot=<0 0 0 1>}
 base_hair_a {parent=head, loc=<8.2703 1.9152 -7.9026>, rot=<0.4301 -0.9 -0.0501 0.0502>}
 dyn_hair_a {parent=base_hair_a, loc=<0 0 0>, rot=<0 0 0 1>}
 dyn_hair_a_02 {parent=dyn_hair_a, loc=<4.4242 0 0>, rot=<-0.0142 0.0987 -0.412 -0.9057>}
 base_hair_b {parent=head, loc=<4.5876 2.6935 -7.7173>, rot=<-0.0107 -0.9989 -0.0303 -0.0355>}
 dyn_hair_b {parent=base_hair_b, loc=<0 0 0>, rot=<0 0 0 1>}
 base_hair_c {parent=head, loc=<8.6069 6.1598 -6.698>, rot=<0.1473 -0.9768 -0.1349 -0.0776>}
 dyn_hair_c {parent=base_hair_c, loc=<0 0 0>, rot=<0 0 0 1>}
 dyn_hair_c_01 {parent=dyn_hair_c, loc=<3.4692 0 0>, rot=<0 0.1439 0 -0.9896>}
 base_hair_d {parent=head, loc=<9.9987 8.83 -4.2923>, rot=<0.0037 -0.8837 -0.4681 0.0019>}
 dyn_hair_d {parent=base_hair_d, loc=<0 0 0>, rot=<0 0 0 1>}
 base_hair_e {parent=head, loc=<9.4877 10.2215 -0.0538>, rot=<-0.0029 0.7071 0.7071 -0.0029>}
 dyn_hair_e {parent=base_hair_e, loc=<0 0 0>, rot=<0 0 0 1>}
 base_hair_f {parent=head, loc=<8.3142 10.2118 3.2559>, rot=<0.0131 0.6999 0.7139 -0.0186>}
 dyn_hair_f {parent=base_hair_f, loc=<0 0 0>, rot=<0 0 0 1>}
 base_hair_h {parent=head, loc=<8.9451 7.8372 5.7561>, rot=<0.0149 0.2796 0.96 -0.0026>}
 dyn_hair_h {parent=base_hair_h, loc=<0 0 0>, rot=<0 0 0 1>}
 base_hair_i {parent=head, loc=<5.1181 -2.5592 8.0966>, rot=<0.0406 0.0793 0.9508 -0.2967>}
 dyn_hair_i {parent=base_hair_i, loc=<0 0 0>, rot=<0 0 0 1>}
 dyn_hair_i_01 {parent=dyn_hair_i, loc=<3.8199 0 0>, rot=<-0.0838 0.0106 0.3069 -0.948>}
 base_hair_j {parent=head, loc=<11.0449 4.3669 -7.7608>, rot=<0.3162 -0.5099 0.0959 -0.7943>}
 dyn_hair_j {parent=base_hair_j, loc=<0 0 0>, rot=<0 0 0 1>}
 base_hair_k {parent=head, loc=<10.3452 5.7279 -7.158>, rot=<0.2365 -0.737 0.081 -0.628>}
 dyn_hair_k {parent=base_hair_k, loc=<0 0 0>, rot=<0 0 0 1>}
 base_hair_l {parent=head, loc=<7.1628 3.3943 7.281>, rot=<0.0086 0.1145 0.9934 -0.0076>}
 dyn_hair_l {parent=base_hair_l, loc=<0 0 0>, rot=<0 0 0 1>}
 dyn_hair_l_01 {parent=dyn_hair_l, loc=<3.1767 0 0>, rot=<0 0 0.0187 -0.9998>}
 base_hair_g {parent=head, loc=<9.1923 10.5552 4.747>, rot=<0.6294 -0.0906 0.7191 -0.28>}
 dyn_hair_g {parent=base_hair_g, loc=<0 0 0>, rot=<0 0 0 1>}
 dyn_hair_g_01 {parent=dyn_hair_g, loc=<3.7498 -0.2182 0>, rot=<0 0 -0.1391 -0.9903>}
 clavicle_l {parent=spine_05, loc=<5.4831 -0.1053 -3.2946>, rot=<0.0465 -0.7289 -0.009 -0.6829>}
 upperarm_l {parent=clavicle_l, loc=<9.8658 -1.2136 -0.4105>, rot=<0.0164 -0.3245 -0.0269 -0.9454>}
 lowerarm_l {parent=upperarm_l, loc=<26.465 -0.8345 0.259>, rot=<0.0036 0.0161 -0.2298 -0.9731>}
 hand_l {parent=lowerarm_l, loc=<22.7355 -0.3203 -0.6105>, rot=<-0.6784 -0.0223 0.0735 -0.7307>}
 thumb_01_l {parent=hand_l, loc=<1.3118 -0.3191 -2.459>, rot=<0.6546 -0.4042 0.0664 -0.6354>}
 thumb_02_l {parent=thumb_01_l, loc=<5.9336 -0.0448 0.0401>, rot=<-0.0037 0.0213 0.298 -0.9543>}
 thumb_03_l {parent=thumb_02_l, loc=<2.2223 0.3682 -0.0907>, rot=<0.0118 -0.013 0.0066 -0.9998>}
 index_metacarpal_l {parent=hand_l, loc=<1.6535 0.0183 -2.96>, rot=<0.0158 -0.0124 -0.0173 -0.9996>}
 index_01_l {parent=index_metacarpal_l, loc=<8.3008 0.0357 -0.0524>, rot=<-0.0044 0.0485 0.0924 -0.9945>}
 index_02_l {parent=index_01_l, loc=<3.3587 -0.3004 0.0452>, rot=<-0.002 0.0027 0.3225 -0.9466>}
 index_03_l {parent=index_02_l, loc=<2.5578 -0.123 -0.0252>, rot=<-0.002 0.0067 0.2558 -0.9667>}
 middle_metacarpal_l {parent=hand_l, loc=<2.1856 -0.1725 -0.5984>, rot=<0.006 0.0168 -0.0329 -0.9993>}
 middle_01_l {parent=middle_metacarpal_l, loc=<7.3275 -0.2447 0.0129>, rot=<-0.0208 0.0042 0.0858 -0.9961>}
 middle_02_l {parent=middle_01_l, loc=<4.0123 0.1401 0.1175>, rot=<-0.0218 0.0258 0.3512 -0.9357>}
 middle_03_l {parent=middle_02_l, loc=<2.6077 0.0534 0.1373>, rot=<0.0082 0.0188 0.1708 -0.9851>}
 ring_metacarpal_l {parent=hand_l, loc=<2.2074 -0.3625 1.3638>, rot=<-0.0284 0.0722 -0.0458 -0.9959>}
 ring_01_l {parent=ring_metacarpal_l, loc=<7.0295 -0.6197 -0.1446>, rot=<-0.0489 -0.0252 0.1892 -0.9804>}
 ring_02_l {parent=ring_01_l, loc=<3.6174 0.4647 -0.1027>, rot=<0 0 0.2806 -0.9598>}
 ring_03_l {parent=ring_02_l, loc=<2.4152 0.114 0.0199>, rot=<0 0 0.1189 -0.9929>}
 pinky_metacarpal_l {parent=hand_l, loc=<1.7616 -0.9753 3.3814>, rot=<-0.0397 0.0814 -0.069 -0.9935>}
 pinky_01_l {parent=pinky_metacarpal_l, loc=<7.1709 -0.4831 -0.1094>, rot=<-0.0654 -0.0035 0.2242 -0.9723>}
 pinky_02_l {parent=pinky_01_l, loc=<2.8274 0.0893 -0.1043>, rot=<0.0037 0.0011 0.2854 -0.9584>}
 pinky_03_l {parent=pinky_02_l, loc=<2.2046 0.0307 -0.0078>, rot=<-0.0002 0.0098 0.1445 -0.9895>}
 deform_wrist_dn_base_l {parent=hand_l, loc=<0 -1.5326 0>, rot=<0 0 0 -1>}
 deform_wrist_dn_l {parent=deform_wrist_dn_base_l, loc=<0 0 0>, rot=<0 0 0 -1>}
 deform_wrist_up_base_l {parent=hand_l, loc=<0 2.3009 0>, rot=<0 0 0 -1>}
 deform_wrist_up_l {parent=deform_wrist_up_base_l, loc=<0 0 0>, rot=<0 0 0 -1>}
 lowerarm_twist_02_l {parent=lowerarm_l, loc=<10.6707 0 0>, rot=<0 0 0 1>}
 lowerarm_twist_01_l {parent=lowerarm_l, loc=<18.316 0 0>, rot=<0 0 0 1>}
 deform_elbow_in_base_l {parent=lowerarm_l, loc=<0.1834 3.6062 -0.2572>, rot=<0 0 0 -1>}
 deform_elbow_in_l {parent=deform_elbow_in_base_l, loc=<0 0 0>, rot=<0 0 0 -1>}
 deform_elbow_out_base_l {parent=lowerarm_l, loc=<0 -4 0>, rot=<0 0 0 1>}
 deform_elbow_out_l {parent=deform_elbow_out_base_l, loc=<0 0 0>, rot=<0 0 0 1>}
 upperarm_twist_01_l {parent=upperarm_l, loc=<7.2135 -0.1134 0>, rot=<0 0 0 1>}
 deform_upperarm_bk_base_l {parent=upperarm_twist_01_l, loc=<-6.4053 -4.3746 0.314>, rot=<0 0 0 -1>}
 deform_upperarm_bk_l {parent=deform_upperarm_bk_base_l, loc=<0 0 0>, rot=<0 0 0 -1>}
 deform_upperarm_fr_base_l {parent=upperarm_twist_01_l, loc=<-5.6312 5.092 0.3354>, rot=<0 0 0 -1>}
 deform_upperarm_fr_l {parent=deform_upperarm_fr_base_l, loc=<0 0 0>, rot=<0 0 0 -1>}
 deform_upperarm_up_base_l {parent=upperarm_twist_01_l, loc=<-8.891 0.4825 3.9517>, rot=<0 0 0 -1>}
 deform_upperarm_up_l {parent=deform_upperarm_up_base_l, loc=<0 0 0>, rot=<0 0 0 -1>}
 upperarm_twist_02_l {parent=upperarm_l, loc=<19.7998 -0.1134 0>, rot=<0 0 0 1>}
 clavicle_r {parent=spine_05, loc=<5.4831 -0.1053 3.2946>, rot=<0.7289 0.0465 -0.6829 0.009>}
 upperarm_r {parent=clavicle_r, loc=<-9.8658 1.2136 0.4105>, rot=<0.0164 -0.3245 -0.0269 -0.9454>}
 lowerarm_r {parent=upperarm_r, loc=<-26.465 0.8345 -0.259>, rot=<0.0036 0.0161 -0.2298 -0.9731>}
 hand_r {parent=lowerarm_r, loc=<-22.7355 0.3203 0.6106>, rot=<-0.6784 -0.0223 0.0735 -0.7307>}
 thumb_01_r {parent=hand_r, loc=<-1.3118 0.3191 2.459>, rot=<0.6546 -0.4042 0.0664 -0.6354>}
 thumb_02_r {parent=thumb_01_r, loc=<-5.9336 0.0448 -0.0401>, rot=<-0.0037 0.0213 0.298 -0.9543>}
 thumb_03_r {parent=thumb_02_r, loc=<-2.2223 -0.3682 0.0907>, rot=<0.0118 -0.013 0.0066 -0.9998>}
 index_metacarpal_r {parent=hand_r, loc=<-1.6535 -0.0183 2.96>, rot=<0.0158 -0.0124 -0.0173 -0.9996>}
 index_01_r {parent=index_metacarpal_r, loc=<-8.3008 -0.0357 0.0524>, rot=<-0.0044 0.0485 0.0924 -0.9945>}
 index_02_r {parent=index_01_r, loc=<-3.3587 0.3004 -0.0452>, rot=<-0.002 0.0027 0.3225 -0.9466>}
 index_03_r {parent=index_02_r, loc=<-2.5578 0.123 0.0252>, rot=<-0.002 0.0067 0.2558 -0.9667>}
 middle_metacarpal_r {parent=hand_r, loc=<-2.1856 0.1725 0.5984>, rot=<0.006 0.0168 -0.0329 -0.9993>}
 middle_01_r {parent=middle_metacarpal_r, loc=<-7.3275 0.2447 -0.0129>, rot=<-0.0208 0.0042 0.0858 -0.9961>}
 middle_02_r {parent=middle_01_r, loc=<-4.0123 -0.1401 -0.1175>, rot=<-0.0218 0.0258 0.3512 -0.9357>}
 middle_03_r {parent=middle_02_r, loc=<-2.6077 -0.0534 -0.1373>, rot=<0.0082 0.0188 0.1708 -0.9851>}
 ring_metacarpal_r {parent=hand_r, loc=<-2.2074 0.3625 -1.3638>, rot=<-0.0284 0.0722 -0.0458 -0.9959>}
 ring_01_r {parent=ring_metacarpal_r, loc=<-7.0295 0.6197 0.1446>, rot=<-0.0489 -0.0252 0.1892 -0.9804>}
 ring_02_r {parent=ring_01_r, loc=<-3.6174 -0.4647 0.1027>, rot=<0 0 0.2806 -0.9598>}
 ring_03_r {parent=ring_02_r, loc=<-2.4152 -0.114 -0.0199>, rot=<0 0 0.1189 -0.9929>}
 pinky_metacarpal_r {parent=hand_r, loc=<-1.7616 0.9753 -3.3814>, rot=<-0.0397 0.0814 -0.069 -0.9935>}
 pinky_01_r {parent=pinky_metacarpal_r, loc=<-7.1709 0.4831 0.1094>, rot=<-0.0654 -0.0035 0.2242 -0.9723>}
 pinky_02_r {parent=pinky_01_r, loc=<-2.8274 -0.0893 0.1043>, rot=<0.0037 0.0011 0.2854 -0.9584>}
 pinky_03_r {parent=pinky_02_r, loc=<-2.2046 -0.0307 0.0078>, rot=<-0.0002 0.0098 0.1445 -0.9895>}
 deform_wrist_dn_base_r {parent=hand_r, loc=<-0.0001 1.5328 0>, rot=<0 0 0 -1>}
 deform_wrist_dn_r {parent=deform_wrist_dn_base_r, loc=<0 0 0>, rot=<0 0 0 -1>}
 deform_wrist_up_base_r {parent=hand_r, loc=<-0.0002 -2.3007 -0.0001>, rot=<0 0 0 -1>}
 deform_wrist_up_r {parent=deform_wrist_up_base_r, loc=<0 0 0>, rot=<0 0 0 -1>}
 lowerarm_twist_02_r {parent=lowerarm_r, loc=<-10.6707 0 0>, rot=<0 0 0 -1>}
 lowerarm_twist_01_r {parent=lowerarm_r, loc=<-18.316 0 0>, rot=<0 0 0 -1>}
 deform_elbow_in_base_r {parent=lowerarm_r, loc=<-0.1834 -3.6062 0.2572>, rot=<0 0 0 -1>}
 deform_elbow_in_r {parent=deform_elbow_in_base_r, loc=<0 0 0>, rot=<0 0 0 -1>}
 deform_elbow_out_base_r {parent=lowerarm_r, loc=<0 4 -0.0001>, rot=<0 0 0 -1>}
 deform_elbow_out_r {parent=deform_elbow_out_base_r, loc=<0 0 0>, rot=<0 0 0 1>}
 upperarm_twist_01_r {parent=upperarm_r, loc=<-7.2135 0.1134 0>, rot=<0 0 0 -1>}
 deform_upperarm_bk_base_r {parent=upperarm_twist_01_r, loc=<6.4053 4.3746 -0.314>, rot=<0 0 0 -1>}
 deform_upperarm_bk_r {parent=deform_upperarm_bk_base_r, loc=<0 0 0>, rot=<0 0 0 -1>}
 deform_upperarm_fr_base_r {parent=upperarm_twist_01_r, loc=<5.6312 -5.092 -0.3354>, rot=<0 0 0 -1>}
 deform_upperarm_fr_r {parent=deform_upperarm_fr_base_r, loc=<0 0 0>, rot=<0 0 0 -1>}
 deform_upperarm_up_base_r {parent=upperarm_twist_01_r, loc=<8.891 -0.4825 -3.9517>, rot=<0 0 0 -1>}
 deform_upperarm_up_r {parent=deform_upperarm_up_base_r, loc=<0 0 0>, rot=<0 0 0 -1>}
 upperarm_twist_02_r {parent=upperarm_r, loc=<-19.7998 0.1134 0>, rot=<0 0 0 -1>}
 attach_backpack {parent=spine_05, loc=<-0.123 -13.473 0>, rot=<-0.0093 -0.707 0.0093 -0.707>}
 attach_cape {parent=spine_05, loc=<9.017 -11.714 0>, rot=<0.0616 -0.7044 -0.0616 -0.7044>}
 head_null {parent=spine_05, type=Null, loc=<20.6573 3.5489 0>, rot=<0 0 -0.036 -0.9994>}
 deform_pec_aim_base_l {parent=spine_04, loc=<3.2456 9.3309 -1.2237>, rot=<0.0603 -0.4033 0.3318 -0.8507>}
 deform_pec_aim_l {parent=deform_pec_aim_base_l, loc=<0 0 0>, rot=<0 0 0 -1>}
 deform_pec_base_l {parent=deform_pec_aim_l, loc=<9.3665 0 0>, rot=<0 0 0 -1>}
 deform_pec_l {parent=deform_pec_base_l, loc=<0 0 0>, rot=<0 0 0 -1>}
 deform_pec_aim_base_r {parent=spine_04, loc=<3.2452 9.3309 1.2237>, rot=<-0.4033 -0.0603 0.8507 0.3318>}
 deform_pec_aim_r {parent=deform_pec_aim_base_r, loc=<0 0 0>, rot=<0 0 0 -1>}
 deform_pec_base_r {parent=deform_pec_aim_r, loc=<-9.3667 -0.0003 -0.0004>, rot=<0 0 0 -1>}
 deform_pec_r {parent=deform_pec_base_r, loc=<0 0 0>, rot=<0 0 0 -1>}
 thigh_l {parent=pelvis, loc=<-1.0855 -0.2109 -8.7053>, rot=<0.0009 0.0355 0.0155 -0.9993>}
 calf_l {parent=thigh_l, loc=<-42.5284 -0.0019 -0.7297>, rot=<0.0006 -0.0161 -0.0391 -0.9991>}
 foot_l {parent=calf_l, loc=<-46.7401 -2.5014 -0.592>, rot=<0.0427 -0.0238 0.0353 -0.9982>}
 ball_l {parent=foot_l, loc=<-3.5658 10.8066 -0.1544>, rot=<0.0001 -0.0001 -0.7063 -0.7079>}
 calf_twist_02_l {parent=calf_l, loc=<-19.1546 -1.7943 -0.2638>, rot=<0.0237 0.0095 -0.0035 -0.9997>}
 calf_twist_01_l {parent=calf_l, loc=<-35.8483 -1.9197 -0.5794>, rot=<0.0237 0.0095 -0.0035 -0.9997>}
 deform_knee_in_base_l {parent=calf_l, loc=<0.8031 -5.0231 0.2456>, rot=<0 0 0 -1>}
 deform_knee_in_l {parent=deform_knee_in_base_l, loc=<0 0 0>, rot=<0 0 0 -1>}
 deform_knee_out_base_l {parent=calf_l, loc=<0 4 0>, rot=<0 0 0 -1>}
 deform_knee_out_l {parent=deform_knee_out_base_l, loc=<0 0 0>, rot=<0 0 0 1>}
 thigh_twist_01_l {parent=thigh_l, loc=<-27.952 0 0>, rot=<0.0013 0 0 -1>}
 deform_groin_base_l {parent=thigh_l, loc=<0.4801 5.6776 -0.2486>, rot=<0 0 0 -1>}
 deform_groin_l {parent=deform_groin_base_l, loc=<0 0 0>, rot=<0 0 0 -1>}
 thigh_r {parent=pelvis, loc=<-1.0855 -0.2109 8.7053>, rot=<0.0355 -0.0009 0.9993 0.0155>}
 calf_r {parent=thigh_r, loc=<42.5284 0.0019 0.7297>, rot=<0.0006 -0.0161 -0.0391 -0.9991>}
 foot_r {parent=calf_r, loc=<46.7401 2.5014 0.592>, rot=<0.0427 -0.0238 0.0353 -0.9982>}
 ball_r {parent=foot_r, loc=<3.5658 -10.8066 0.1544>, rot=<0.0001 -0.0001 -0.7063 -0.7079>}
 calf_twist_02_r {parent=calf_r, loc=<19.1546 1.7943 0.2638>, rot=<0.0237 0.0095 -0.0035 -0.9997>}
 calf_twist_01_r {parent=calf_r, loc=<35.8483 1.9197 0.5794>, rot=<0.0237 0.0095 -0.0035 -0.9997>}
 deform_knee_in_base_r {parent=calf_r, loc=<-0.8031 5.0231 -0.2457>, rot=<0 0 0 -1>}
 deform_knee_in_r {parent=deform_knee_in_base_r, loc=<0 0 0>, rot=<0 0 0 -1>}
 deform_knee_out_base_r {parent=calf_r, loc=<0 -4 0>, rot=<0 0 0 -1>}
 deform_knee_out_r {parent=deform_knee_out_base_r, loc=<0 0 0>, rot=<0 0 0 1>}
 thigh_twist_01_r {parent=thigh_r, loc=<27.952 0 0>, rot=<0.0013 0 0 -1>}
 deform_groin_base_r {parent=thigh_r, loc=<-0.4801 -5.6776 0.2486>, rot=<0 0 0 -1>}
 deform_groin_r {parent=deform_groin_base_r, loc=<0 0 0>, rot=<0 0 0 -1>}
 deform_glute_base_l {parent=pelvis, loc=<-7.3466 -9.0213 -9.1245>, rot=<0.0172 0.031 -0.4477 -0.8935>}
 deform_glute_l {parent=deform_glute_base_l, loc=<0 0 0>, rot=<0 0 0 1>}
 deform_glute_base_r {parent=pelvis, loc=<-7.3464 -9.0213 9.1245>, rot=<0.031 -0.0172 0.8935 -0.4477>}
 deform_glute_r {parent=deform_glute_base_r, loc=<0 0 0>, rot=<0 0 0 1>}
]
variables [
 Cap crouch {type=bool, default=False}
 Crouch walk {type=double, default=0}
 Front end {type=bool, default=False}
 Targeting {type=double, default=0}
]
