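# Physics asset for the TimeTrotter female medium face-accessory mesh, setting up dangling earrings and a seven-link ponytail.
# Small base capsules lock both earrings and the ponytail to the head, and the dynamic bones hang from them with limited swing and zero twist.
# Capsules on the head, neck, upper spine, clavicles and upper arms form body volumes, each constrained to spine_04 with free angular motion.
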PhysicsAsset /BRCosmetics/Characters/Player/Female/Medium/Bodies/F_MED_TimeTrotter/Meshes/Parts/F_MED_TimeTrotter_FaceAcc_Physics
bodies {
 base_earring_l [{shapeType=Capsule, length=0.505, radius=0.505, shapeName=base_earring_l_capsule, rotation=<rot 0 -90 0>}]
 base_earring_r [{shapeType=Capsule, length=0.505, radius=0.505, shapeName=base_earring_r_capsule, rotation=<rot 0 -90 0>}]
 base_ponytail [{shapeType=Capsule, length=0.505, radius=0.505, shapeName=base_ponytail_capsule, rotation=<rot 0 -90 0>}]
 clavicle_l [{shapeType=Capsule, length=12.6283, radius=7.5, shapeName=clavicle_l_capsule, center=<5 0 -1.6229>, rotation=<rot 0 -90 90>}]
 clavicle_r [{shapeType=Capsule, length=12.2586, radius=7.5, shapeName=clavicle_r_capsule, center=<-5 0 1.6229>, rotation=<rot 0 -90 -90>}]
 dyn_earring_l [{shapeType=Capsule, length=1.9538, radius=1.5, shapeName=dyn_earring_l_capsule, center=<2 0 0.5>, rotation=<rot 0 90 90>}]
 dyn_earring_r [{shapeType=Capsule, length=1.954, radius=1.5, shapeName=dyn_earring_r_capsule, center=<2 0 0.4>, rotation=<rot 0 90 90>}]
 dyn_ponytail_01 [{shapeType=Capsule, length=2, radius=4, shapeName=dyn_ponytail_01_capsule, center=<5 0 0>, rotation=<rot -90 0 0>}]
 dyn_ponytail_02 [{shapeType=Capsule, length=8.6515, radius=7.2051, shapeName=dyn_ponytail_02_capsule, center=<4 0 0>, rotation=<rot 0 0 -90>}]
 dyn_ponytail_03 [{shapeType=Capsule, length=9.2583, radius=6.7534, shapeName=dyn_ponytail_03_capsule, center=<4 0 0>, rotation=<rot 0 0 -90>}]
 dyn_ponytail_04 [{shapeType=Capsule, length=8.5321, radius=6.3105, shapeName=dyn_ponytail_04_capsule, center=<5.4576 0 0>, rotation=<rot 0 0 -90>}]
 dyn_ponytail_05 [{shapeType=Capsule, length=8.5478, radius=4.9281, shapeName=dyn_ponytail_05_capsule, center=<4 0 0>, rotation=<rot 0 0 -90>}]
 dyn_ponytail_06 [{shapeType=Capsule, length=6.2191, radius=4.1154, shapeName=dyn_ponytail_06_capsule, center=<4 0 1>, rotation=<rot 0 0 -90>}]
 dyn_ponytail_07 [{shapeType=Capsule, length=6.0402, radius=3.9598, shapeName=dyn_ponytail_07_capsule, center=<4 0 0>, rotation=<rot 0 0 -90>}]
 head [{shapeType=Capsule, length=5.1639, radius=7.3118, shapeName=head_capsule, center=<5.3009 -2.0272 0>, rotation=<rot 90 0 0>}]
 neck_01 [{shapeType=Capsule, length=6.84, radius=6.9326, shapeName=neck_01_capsule, center=<4 0 0>, rotation=<rot 0 -110 -90>}]
 spine_04 [{shapeType=Capsule, length=20, radius=10, shapeName=spine_04_capsule, center=<-5.9766 -1.1982 0>, rotation=<rot 0 -90.0016 0>}]
 spine_05 [{shapeType=Capsule, length=17.8791, radius=10.5, shapeName=spine_05_capsule, center=<1.9629 1.9752 0>, rotation=<rot 0 -90 0>}]
 upperarm_l [{shapeType=Capsule, length=17.6837, radius=7.6053, shapeName=upperarm_l_capsule, center=<10 0 0>, rotation=<rot 0 -90 90>}]
 upperarm_r [{shapeType=Capsule, length=18.7079, radius=6.5811, shapeName=upperarm_r_capsule, center=<-10 -1.0716 0>, rotation=<rot 0 -90 -90>}]
}
constraints [
 {
  bone1Name=base_ponytail
  bone2Name=head
  swing1LimitDegrees=45
  swing1Motion=Locked
  swing2LimitDegrees=45
  swing2Motion=Locked
  twistLimitDegrees=45
  twistMotion=Locked
 }
 {
  bone1Name=base_earring_l
  bone2Name=head
  swing1LimitDegrees=45
  swing1Motion=Locked
  swing2LimitDegrees=45
  swing2Motion=Locked
  twistLimitDegrees=45
  twistMotion=Locked
 }
 {
  bone1Name=base_earring_r
  bone2Name=head
  swing1LimitDegrees=45
  swing1Motion=Locked
  swing2LimitDegrees=45
  swing2Motion=Locked
  twistLimitDegrees=45
  twistMotion=Locked
 }
 {bone1Name=dyn_earring_l, bone2Name=base_earring_l, swing1LimitDegrees=40, swing2LimitDegrees=20, twistLimitDegrees=0}
 {bone1Name=dyn_earring_r, bone2Name=base_earring_r, swing1LimitDegrees=40, swing2LimitDegrees=20, twistLimitDegrees=0}
 {bone1Name=dyn_ponytail_01, bone2Name=base_ponytail, swing1LimitDegrees=40, swing2LimitDegrees=2, twistLimitDegrees=0}
 {bone1Name=dyn_ponytail_02, bone2Name=dyn_ponytail_01, swing1LimitDegrees=45, swing2LimitDegrees=45, twistLimitDegrees=0}
 {bone1Name=dyn_ponytail_03, bone2Name=dyn_ponytail_02, swing1LimitDegrees=45, swing2LimitDegrees=45, twistLimitDegrees=0}
 {bone1Name=dyn_ponytail_04, bone2Name=dyn_ponytail_03, swing1LimitDegrees=45, swing2LimitDegrees=45, twistLimitDegrees=0}
 {bone1Name=dyn_ponytail_05, bone2Name=dyn_ponytail_04, swing1LimitDegrees=45, swing2LimitDegrees=45, twistLimitDegrees=0}
 {bone1Name=dyn_ponytail_06, bone2Name=dyn_ponytail_05, swing1LimitDegrees=45, swing2LimitDegrees=45, twistLimitDegrees=0}
 {bone1Name=dyn_ponytail_07, bone2Name=dyn_ponytail_06, swing1LimitDegrees=45, swing2LimitDegrees=45, twistLimitDegrees=0}
 {
  bone1Name=spine_05
  bone2Name=spine_04
  swing1LimitDegrees=45
  swing1Motion=Free
  swing2LimitDegrees=45
  swing2Motion=Free
  twistLimitDegrees=45
  twistMotion=Free
 }
 {
  bone1Name=neck_01
  bone2Name=spine_04
  swing1LimitDegrees=45
  swing1Motion=Free
  swing2LimitDegrees=45
  swing2Motion=Free
  twistLimitDegrees=45
  twistMotion=Free
 }
 {
  bone1Name=head
  bone2Name=spine_04
  swing1LimitDegrees=45
  swing1Motion=Free
  swing2LimitDegrees=45
  swing2Motion=Free
  twistLimitDegrees=45
  twistMotion=Free
 }
 {
  bone1Name=clavicle_l
  bone2Name=spine_04
  swing1LimitDegrees=45
  swing1Motion=Free
  swing2LimitDegrees=45
  swing2Motion=Free
  twistLimitDegrees=45
  twistMotion=Free
 }
 {
  bone1Name=upperarm_l
  bone2Name=spine_04
  swing1LimitDegrees=45
  swing1Motion=Free
  swing2LimitDegrees=45
  swing2Motion=Free
  twistLimitDegrees=45
  twistMotion=Free
 }
 {
  bone1Name=clavicle_r
  bone2Name=spine_04
  swing1LimitDegrees=45
  swing1Motion=Free
  swing2LimitDegrees=45
  swing2Motion=Free
  twistLimitDegrees=45
  twistMotion=Free
 }
 {
  bone1Name=upperarm_r
  bone2Name=spine_04
  swing1LimitDegrees=45
  swing1Motion=Free
  swing2LimitDegrees=45
  swing2Motion=Free
  twistLimitDegrees=45
  twistMotion=Free
 }
]
